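# Control Rig for the F_MED_Parcel_Petal female body part that nudges back hair bones as the head and right arm move.
# Spherical pose readers on head and upperarm_r write alphas, some multiplied into combined look-and-aim weights.
# These weights and another head-versus-spine_05 reader drive Modify Transforms offsets on the dyn_hair_back base bones, some smoothed by Alpha Interp.
# Several alphas, including look left/right and skydiving, drive no bones.
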ControlRigBlueprint /BRCosmetics/Characters/Player/Female/Medium/Bodies/F_MED_Parcel_Petal/Meshes/Parts/F_MED_Parcel_Petal_CtrlRig1
graphs [
 ContainedGraph {nodes [
   n1 = RigVMFunctionEntryNode()
   SequenceExecution_7 = &RigVMFunction_Sequence(ExecuteContext=@n1.ExecuteContext)
   SequenceExecution_7_1 = &RigVMFunction_Sequence(ExecuteContext=@SequenceExecution_7.B)
   SequenceExecution_7_2 = &RigVMFunction_Sequence(ExecuteContext=@SequenceExecution_7_1.B)
   RigVMFunctionReturnNode(A=@SequenceExecution_7.A, B=@SequenceExecution_7_1.A, C=@SequenceExecution_7_2.A, D=@SequenceExecution_7_2.B)
  ]}
 ContainedGraph {nodes [
   n1 = RigVMFunctionEntryNode()
   SequenceExecution_2 = &RigVMFunction_Sequence(ExecuteContext=@n1.ExecuteContext)
   SequenceExecution_2_1 = &RigVMFunction_Sequence(ExecuteContext=@SequenceExecution_2.B)
   SequenceExecution_2_2 = &RigVMFunction_Sequence(ExecuteContext=@SequenceExecution_2_1.B)
   RigVMFunctionReturnNode(A=@SequenceExecution_2.A, B=@SequenceExecution_2_1.A, C=@SequenceExecution_2_2.A, D=@SequenceExecution_2_2.B)
  ]}
 RigVMFunctionLibrary {links [], nodes []}
 RigVMFunction_Sequence_3_ContainedGraph {nodes [
   n1 = RigVMFunctionEntryNode()
   n2 = &RigVMFunction_Sequence(ExecuteContext=@n1.ExecuteContext)
   n3 = &RigVMFunction_Sequence(ExecuteContext=@n2.B)
   RigVMFunctionReturnNode(B=@n3.A, C=@n3.B, A=@n2.A)
  ]}
 RigVMFunction_Sequence_5_ContainedGraph {nodes [
   n1 = RigVMFunctionEntryNode()
   n2 = &RigVMFunction_Sequence(ExecuteContext=@n1.ExecuteContext)
   n3 = &RigVMFunction_Sequence(ExecuteContext=@n2.B)
   RigVMFunctionReturnNode(B=@n3.A, C=@n3.B, A=@n2.A)
  ]}
 RigVMModel {nodes [
   RigUnit_BeginExecution = &RigUnit_BeginExecution()
   n1 = &RigVMFunction_Sequence(ExecuteContext=@RigUnit_BeginExecution.ExecutePin)
   n2 = &RigVMFunction_Sequence(ExecuteContext=@n1.A)
   RigVMFunction_Sequence = &RigVMFunction_Sequence(ExecuteContext=@n2.A)
   SequenceExecution_7 = RigVMAggregateNode(ExecuteContext=@RigVMFunction_Sequence.A)
   n3 = &RigUnit_SphericalPoseReader(
    ActiveRegionScaleFactors=(PositiveWidth=1,NegativeWidth=1,PositiveHeight=1,NegativeHeight=1)
    ActiveRegionSize=0.065
    Debug=(bDrawDebug=false,DebugScale=25,DebugSegments=20,DebugThickness=0.25)
    DriverAxis=(X=-1,Y=0,Z=0)
    DriverItem=(Type=Bone,Name="upperarm_r")
    FalloffRegionScaleFactors=(PositiveWidth=1,NegativeWidth=1,PositiveHeight=1,NegativeHeight=1)
    FalloffSize=0.125
    OptionalParentItem=(Type=Bone,Name="None")
    RotationOffset=(X=60,Y=0,Z=65)
    ExecutePin=@SequenceExecution_7.A
   )
   > RigVMVariableNode(Value=@n3.OutputParam) {variable=upperarm_r_SPR_back_alpha}
   n4 = &RigUnit_SphericalPoseReader(
    ActiveRegionScaleFactors=(PositiveWidth=1,NegativeWidth=1,PositiveHeight=1,NegativeHeight=1)
    ActiveRegionSize=0.06
    Debug=(bDrawDebug=false,DebugScale=25,DebugSegments=20,DebugThickness=0.25)
    DriverAxis=(X=-1,Y=0,Z=0)
    DriverItem=(Type=Bone,Name="upperarm_r")
    FalloffRegionScaleFactors=(PositiveWidth=1,NegativeWidth=1,PositiveHeight=1,NegativeHeight=1)
    FalloffSize=0.15
    OptionalParentItem=(Type=Bone,Name="spine_05")
    RotationOffset=(X=-90,Y=0,Z=40)
    ExecutePin=@SequenceExecution_7.B
   )
   > RigVMVariableNode(Value=@n4.OutputParam) {variable=uppearm_r_SPR_aim_up_two_hands_alpha}
   n5 = &RigUnit_SphericalPoseReader(
    ActiveRegionScaleFactors=(PositiveWidth=2,NegativeWidth=2,PositiveHeight=1,NegativeHeight=1)
    ActiveRegionSize=0.05
    Debug=(bDrawDebug=false,DebugScale=25,DebugSegments=20,DebugThickness=0.25)
    DriverAxis=(X=-1,Y=0,Z=0)
    DriverItem=(Type=Bone,Name="upperarm_r")
    FalloffRegionScaleFactors=(PositiveWidth=1,NegativeWidth=1,PositiveHeight=0.75,NegativeHeight=0.75)
    FalloffSize=0.2
    FlipWidthScaling=false
    OptionalParentItem=(Type=Bone)
    RotationOffset=(X=15,Y=150,Z=0)
    ExecutePin=@SequenceExecution_7.C
   )
   > RigVMVariableNode(Value=@n5.OutputParam) {variable=upperarm_r_skydiving}
   n6 = &RigUnit_SphericalPoseReader(
    ActiveRegionScaleFactors=(PositiveWidth=1.2,NegativeWidth=1,PositiveHeight=1,NegativeHeight=1.5)
    ActiveRegionSize=0.075
    Debug=(bDrawDebug=false,DebugScale=25,DebugSegments=20,DebugThickness=0.25)
    DriverAxis=(X=-1,Y=0,Z=0)
    DriverItem=(Type=Bone,Name="upperarm_r")
    FalloffRegionScaleFactors=(PositiveWidth=1,NegativeWidth=1,PositiveHeight=1,NegativeHeight=1)
    FalloffSize=0.2
    OptionalParentItem=(Type=Bone)
    RotationOffset=(X=-60,Y=0,Z=-15)
    ExecutePin=@SequenceExecution_7.D
   )
   > RigVMVariableNode(Value=@n6.OutputParam) {variable=upperarm_r_rifle_up_alpha}
   SequenceExecution_2 = RigVMAggregateNode(ExecuteContext=@RigVMFunction_Sequence.B)
   n7 = &RigUnit_SphericalPoseReader(
    ActiveRegionScaleFactors=(PositiveWidth=1,NegativeWidth=0.5,PositiveHeight=1.5,NegativeHeight=1.5)
    ActiveRegionSize=0.1
    Debug=(bDrawDebug=false,DebugScale=25,DebugSegments=20,DebugThickness=0.25)
    DriverAxis=(X=0,Y=1,Z=0)
    DriverItem=(Type=Bone,Name="head")
    FalloffRegionScaleFactors=(PositiveWidth=1,NegativeWidth=0.7,PositiveHeight=1.5,NegativeHeight=1.5)
    FalloffSize=0.4
    FlipWidthScaling=false
    OptionalParentItem=(Type=Bone)
    RotationOffset=(X=90,Y=0,Z=-60)
    ExecutePin=@SequenceExecution_2.A
   )
   > RigVMVariableNode(Value=@n7.OutputParam) {variable=head_SPR_look_up_alpha}
   n8 = &RigUnit_SphericalPoseReader(
    ActiveRegionScaleFactors=(PositiveWidth=0.75,NegativeWidth=0.75,PositiveHeight=1,NegativeHeight=0.5)
    ActiveRegionSize=0.1
    Debug=(bDrawDebug=false,DebugScale=25,DebugSegments=20,DebugThickness=0.25)
    DriverAxis=(X=0,Y=1,Z=0)
    DriverItem=(Type=Bone,Name="head")
    FalloffRegionScaleFactors=(PositiveWidth=1,NegativeWidth=1,PositiveHeight=1,NegativeHeight=0.7)
    FalloffSize=0.2
    FlipHeightScaling=true
    FlipWidthScaling=false
    OptionalParentItem=(Type=Bone)
    RotationOffset=(X=60,Y=0,Z=0)
    ExecutePin=@SequenceExecution_2.D
   )
   > RigVMVariableNode(Value=@n8.OutputParam) {variable=head_SPR_look_right_alpha}
   n9 = &RigUnit_SphericalPoseReader(
    ActiveRegionScaleFactors=(PositiveWidth=2,NegativeWidth=0.5,PositiveHeight=2,NegativeHeight=2)
    ActiveRegionSize=0.075
    Debug=(bDrawDebug=false,DebugScale=25,DebugSegments=20,DebugThickness=0.25)
    DriverAxis=(X=0,Y=1,Z=0)
    DriverItem=(Type=Bone,Name="head")
    FalloffRegionScaleFactors=(PositiveWidth=1,NegativeWidth=0.7,PositiveHeight=1,NegativeHeight=1)
    FalloffSize=0.2
    FlipHeightScaling=false
    FlipWidthScaling=true
    OptionalParentItem=(Type=Bone)
    RotationOffset=(X=90,Y=0,Z=12.5)
    ExecutePin=@SequenceExecution_2.B
   )
   > RigVMVariableNode(Value=@n9.OutputParam) {variable=head_SPR_look_down_alpha}
   n10 = &RigUnit_SphericalPoseReader(
    ActiveRegionScaleFactors=(PositiveWidth=0.75,NegativeWidth=0.75,PositiveHeight=1,NegativeHeight=0.5)
    ActiveRegionSize=0.1
    Debug=(bDrawDebug=false,DebugScale=25,DebugSegments=20,DebugThickness=0.25)
    DriverAxis=(X=0,Y=1,Z=0)
    DriverItem=(Type=Bone,Name="head")
    FalloffRegionScaleFactors=(PositiveWidth=1,NegativeWidth=1,PositiveHeight=1,NegativeHeight=0.7)
    FalloffSize=0.2
    OptionalParentItem=(Type=Bone)
    RotationOffset=(X=120,Y=0,Z=0)
    ExecutePin=@SequenceExecution_2.C
   )
   > RigVMVariableNode(Value=@n10.OutputParam) {variable=head_SPR_look_left_alpha}
   RigVMFunction_Sequence_2 = RigVMAggregateNode(ExecuteContext=@n2.B)
   RigVMVariableNode(
    Value=&RigVMFunction_MathFloatMul(B=RigVMVariableNode() {variable=uppearm_r_SPR_aim_up_two_hands_alpha}.Value, A=RigVMVariableNode() {variable=head_SPR_look_up_alpha}.Value) {template="Multiply::Execute(in A,in B,out Result)"}.Result
    ExecuteContext=@RigVMFunction_Sequence_2.A
   ) {variable=look_up_and_left_two_hands_multiplied_alpha}
   RigVMVariableNode(
    Value=&RigVMFunction_MathFloatMul(A=RigVMVariableNode() {variable=head_SPR_look_up_alpha}.Value, B=RigVMVariableNode() {variable=upperarm_r_rifle_up_alpha}.Value) {template="Multiply::Execute(in A,in B,out Result)"}.Result
    ExecuteContext=@RigVMFunction_Sequence_2.B
   ) {variable=look_up_two_hands_rifle_multiplied_alpha}
   RigVMVariableNode(
    Value=&RigVMFunction_MathFloatMul(B=RigVMVariableNode() {variable=head_SPR_look_down_alpha}.Value, A=RigVMVariableNode() {variable=upperarm_r_SPR_back_alpha}.Value) {template="Multiply::Execute(in A,in B,out Result)"}.Result
    ExecuteContext=@RigVMFunction_Sequence_2.C
   ) {variable=look_down_upperarm_r_back_multiplied_alpha}
   RigVMFunction_Sequence_4 = RigVMAggregateNode(ExecuteContext=@n1.B)
   n11 = &RigUnit_SphericalPoseReader(
    ActiveRegionScaleFactors=(PositiveWidth=2,NegativeWidth=0.5,PositiveHeight=2,NegativeHeight=2)
    ActiveRegionSize=0.075
    Debug=(bDrawDebug=false,DebugScale=25,DebugSegments=20,DebugThickness=0.25)
    DriverAxis=(X=1,Y=0,Z=0)
    DriverItem=(Type=Bone,Name="head")
    FalloffRegionScaleFactors=(PositiveWidth=1,NegativeWidth=0.4,PositiveHeight=1,NegativeHeight=1)
    FalloffSize=0.25
    FlipHeightScaling=false
    FlipWidthScaling=true
    OptionalParentItem=(Type=Bone,Name="spine_05")
    RotationOffset=(X=-70,Y=5,Z=0)
    ExecutePin=@RigVMFunction_Sequence_4.C
   )
   n12 = &RigVMFunction_AlphaInterp(InterpSpeedDecreasing=10, InterpSpeedIncreasing=10, Scale=1, bClampResult=true, Value=@n11.OutputParam)
   > &RigUnit_ModifyTransforms(
    ItemToModify="((Item=(Type=Bone,Name=\"dyn_hair_back_m_base\"),Transform=(Rotation=(X=0,Y=0,Z=0,W=1),Translation=(X=3,Y=-3,Z=-2),Scale3D=(X=1,Y=1,Z=1))),(Item=(Type=Bone,Name=\"dyn_hair_back_l_base\"),Transform=(Rotation=(X=0,Y=0,Z=0,W=1),Translation=(X=1,Y=-5,Z=-2),Scale3D ...(truncated)"
    Weight=@n12.Result
    ItemToModify.0.Transform.Rotation=&RigVMFunction_MathQuaternionFromEuler(Euler=(X=0,Y=0,Z=0), RotationOrder=XYZ).Result
    ItemToModify.2.Transform.Rotation=&RigVMFunction_MathQuaternionFromEuler(Euler=(X=0,Y=0,Z=0), RotationOrder=XYZ).Result
    ItemToModify.1.Transform.Rotation=&RigVMFunction_MathQuaternionFromEuler(Euler=(X=0,Y=0,Z=0), RotationOrder=XYZ).Result
   )
   &RigUnit_ModifyTransforms(
    ItemToModify=((Item=(Type=Bone,Name="dyn_hair_back_r_base"),Transform=(Rotation=(X=0,Y=0,Z=0,W=1),Translation=(X=1,Y=2,Z=-1),Scale3D=(X=1,Y=1,Z=1))))
    ItemToModify.0.Transform.Rotation=&RigVMFunction_MathQuaternionFromEuler(Euler=(X=0,Y=5,Z=-5), RotationOrder=XYZ).Result
    Weight=RigVMVariableNode() {variable=look_up_and_left_two_hands_multiplied_alpha}.Value
    ExecutePin=@RigVMFunction_Sequence_4.B
   )
   > &RigUnit_ModifyTransforms(
    ItemToModify=((Item=(Type=Bone,Name="dyn_hair_back_m_base"),Transform=(Rotation=(X=0,Y=0,Z=0,W=1),Translation=(X=0,Y=2,Z=-2),Scale3D=(X=1,Y=1,Z=1))))
    ItemToModify.0.Transform.Rotation=&RigVMFunction_MathQuaternionFromEuler(Euler=(X=0,Y=5,Z=0), RotationOrder=XYZ).Result
    Weight=RigVMVariableNode() {variable=look_up_and_left_two_hands_multiplied_alpha}.Value
   )
   > &RigUnit_ModifyTransforms(
    ItemToModify=((Item=(Type=Bone,Name="dyn_hair_back_l_base"),Transform=(Rotation=(X=0,Y=0,Z=0,W=1),Translation=(X=-0.5,Y=-2,Z=-4),Scale3D=(X=1,Y=1,Z=1))))
    ItemToModify.0.Transform.Rotation=&RigVMFunction_MathQuaternionFromEuler(Euler=(X=0,Y=0,Z=-5), RotationOrder=XYZ).Result
    Weight=RigVMVariableNode() {variable=look_up_and_left_two_hands_multiplied_alpha}.Value
   )
   &RigUnit_ModifyTransforms(
    ItemToModify="((Item=(Type=Bone,Name=\"dyn_hair_back_m_base\"),Transform=(Rotation=(X=0,Y=0,Z=0,W=1),Translation=(X=2,Y=-4,Z=0),Scale3D=(X=1,Y=1,Z=1))),(Item=(Type=Bone,Name=\"dyn_hair_back_l_base\"),Transform=(Rotation=(X=0,Y=0,Z=0,W=1),Translation=(X=2,Y=-4,Z=0),Scale3D=( ...(truncated)"
    ItemToModify.0.Transform.Rotation=&RigVMFunction_MathQuaternionFromEuler(Euler=(X=0,Y=0,Z=-5), RotationOrder=XYZ).Result
    ItemToModify.2.Transform.Rotation=&RigVMFunction_MathQuaternionFromEuler(Euler=(X=0,Y=0,Z=-5), RotationOrder=XYZ).Result
    ItemToModify.1.Transform.Rotation=&RigVMFunction_MathQuaternionFromEuler(Euler=(X=0,Y=0,Z=-5), RotationOrder=XYZ).Result
    Weight=RigVMVariableNode() {variable=head_SPR_look_up_alpha}.Value
    ExecutePin=@RigVMFunction_Sequence_4.A
   )
   AlphaInterp = &RigVMFunction_AlphaInterp(
    InterpSpeedDecreasing=10
    InterpSpeedIncreasing=10
    Scale=1
    bClampResult=true
    Value=/Script/RigVM.RigVMFunction_MathFloatMax(
     B=RigVMVariableNode() {variable=look_up_two_hands_rifle_multiplied_alpha}.Value
     A=RigVMVariableNode() {variable=look_up_and_left_two_hands_multiplied_alpha}.Value
    ) {is=struct RigVMUnitNode, template="Maximum::Execute(in A,in B,out Result)"}.Result
   )
   > &RigUnit_ModifyTransforms(
    ItemToModify="((Item=(Type=Bone,Name=\"dyn_hair_back_m_base\"),Transform=(Rotation=(X=0,Y=0,Z=0,W=1),Translation=(X=4,Y=-3,Z=0),Scale3D=(X=1,Y=1,Z=1))),(Item=(Type=Bone,Name=\"dyn_hair_back_l_base\"),Transform=(Rotation=(X=0,Y=0,Z=0,W=1),Translation=(X=3,Y=-4,Z=0),Scale3D=( ...(truncated)"
    Weight=@AlphaInterp.Result
    ItemToModify.0.Transform.Rotation=&RigVMFunction_MathQuaternionFromEuler(Euler=(X=0,Y=0,Z=0), RotationOrder=XYZ).Result
    ItemToModify.2.Transform.Rotation=&RigVMFunction_MathQuaternionFromEuler(Euler=(X=0,Y=-5,Z=0), RotationOrder=XYZ).Result
    ItemToModify.1.Transform.Rotation=&RigVMFunction_MathQuaternionFromEuler(Euler=(X=0,Y=5,Z=0), RotationOrder=XYZ).Result
   )
  ]}
]
hierarchy [
 root {loc=<0 0 0>, rot=<0 0 0 1>}
 pelvis {parent=root, loc=<0 -0.5301 98.0571>, rot=<0.0127 0.707 -0.0127 -0.707>}
 spine_01 {parent=pelvis, loc=<6.1208 -0.9328 0>, rot=<0 0 0.0576 -0.9983>}
 spine_02 {parent=spine_01, loc=<6.6022 -0.0153 0>, rot=<0 0 0.0022 -1>}
 spine_03 {parent=spine_02, loc=<6.4919 -0.0105 0>, rot=<0 0 0.0015 -1>}
 spine_04 {parent=spine_03, loc=<6.9596 -0.007 0>, rot=<0 0 0.0009 -1>}
 spine_05 {parent=spine_04, loc=<6.9981 -0.0017 0>, rot=<0 0 -0.0042 -1>}
 neck_01 {parent=spine_05, loc=<11.3138 0.1529 0>, rot=<0 0 -0.2003 -0.9797>}
 neck_02 {parent=neck_01, loc=<4.5627 0 0>, rot=<0 0 0.0504 -0.9987>}
 head {parent=neck_02, loc=<5.3914 0 0>, rot=<0 0 0.115 -0.9934>}
 dyn_hair_bang_l_base {parent=head, loc=<3.5485 -0.758 -9.956>, rot=<0.0029 -0.7071 -0.0029 -0.7071>}
 dyn_hair_bang_l_parent {parent=dyn_hair_bang_l_base, loc=<0 0 0>, rot=<0 0 0 1>}
 dyn_hair_side_r_base {parent=head, loc=<0.6982 2.5017 8.8512>, rot=<0.0109 -0.9768 0.2076 0.0512>}
 dyn_hair_side_r_parent {parent=dyn_hair_side_r_base, loc=<0 0 0>, rot=<0 0 0 1>}
 dyn_hair_side_r_01 {parent=dyn_hair_side_r_parent, loc=<3.5248 -0.0627 0.0788>, rot=<0.0697 0.0174 0.0012 -0.9974>}
 dyn_hair_side_r_02 {parent=dyn_hair_side_r_01, loc=<4.1295 -0.0009 0.31>, rot=<0 0 0 1>}
 dyn_hair_side_r_03 {parent=dyn_hair_side_r_02, loc=<3.6307 -0.0223 0.5031>, rot=<0 0 0 1>}
 dyn_hair_back_r_base {parent=head, loc=<3.1226 -5.6768 6.5996>, rot=<0.0813 -0.87 -0.475 0.1044>}
 dyn_hair_back_r_parent {parent=dyn_hair_back_r_base, loc=<0 0 0>, rot=<0 0 0 1>}
 dyn_hair_back_r_01 {parent=dyn_hair_back_r_parent, loc=<5.3186 -0.5291 1.3299>, rot=<0 0 0 1>}
 dyn_hair_back_r_02 {parent=dyn_hair_back_r_01, loc=<3.9918 -0.9775 -0.3881>, rot=<0 0 0 1>}
 dyn_hair_back_r_03 {parent=dyn_hair_back_r_02, loc=<4.1773 -1.0541 0.477>, rot=<0 0 0 1>}
 dyn_hair_back_r_04 {parent=dyn_hair_back_r_03, loc=<5.6033 -1.1159 -1.8734>, rot=<0 0 0 1>}
 dyn_hair_back_r_06 {parent=dyn_hair_back_r_04, loc=<4.0484 3.2364 1.3059>, rot=<0 0.2588 0 -0.9659>}
 dyn_hair_back_r_05 {parent=dyn_hair_back_r_04, loc=<3.2529 0.67 1.9325>, rot=<-0.0153 0.3029 0.07 -0.9503>}
 dyn_hair_leaf_03 {parent=head, loc=<10.5836 1.747 -7.8712>, rot=<-0.1726 0.0542 0.5575 -0.8102>}
 dyn_hair_back_m_base {parent=head, loc=<3.3054 -8.526 -0.3617>, rot=<-0.1471 0.6397 0.749 -0.0905>}
 dyn_hair_back_m_parent {parent=dyn_hair_back_m_base, loc=<0 0 0>, rot=<0 0 0 1>}
 dyn_hair_back_m_01 {parent=dyn_hair_back_m_parent, loc=<0 0 0>, rot=<0 0 0 1>}
 dyn_hair_back_m_02 {parent=dyn_hair_back_m_01, loc=<4.0992 0.3557 0.3929>, rot=<-0.0872 0 0 -0.9962>}
 dyn_hair_back_m_03 {parent=dyn_hair_back_m_02, loc=<4.1531 0.4067 0.5332>, rot=<-0.0436 -0.0019 -0.0436 -0.9981>}
 dyn_hair_back_m_04 {parent=dyn_hair_back_m_03, loc=<5.2163 -0.1838 0.7025>, rot=<0 0 0 1>}
 dyn_hair_back_m_05 {parent=dyn_hair_back_m_04, loc=<3.391 0.091 1.3159>, rot=<0 0 0 1>}
 dyn_hair_back_m_06 {parent=dyn_hair_back_m_05, loc=<4.448 4.8265 -1.2215>, rot=<0.0427 0.2077 0.0091 -0.9772>}
 dyn_hair_back_m_07 {parent=dyn_hair_back_m_05, loc=<6.3451 -1.1559 1.7064>, rot=<0 0 0.0698 -0.9976>}
 dyn_hair_leaf_04 {parent=head, loc=<5.4668 -1.8989 -9.1273>, rot=<-0.176 -0.0893 0.7352 -0.6485>}
 dyn_hair_back_l_base {parent=head, loc=<4.3519 -4.1845 -6.9958>, rot=<-0.0821 0.2488 0.9549 -0.1398>}
 dyn_hair_back_l_parent {parent=dyn_hair_back_l_base, loc=<0 0 0>, rot=<0 0 0 1>}
 dyn_hair_back_l_01 {parent=dyn_hair_back_l_parent, loc=<4.9044 -0.8975 1.5243>, rot=<0 0 0 1>}
 dyn_hair_back_l_02 {parent=dyn_hair_back_l_01, loc=<3.877 0.8012 -0.4498>, rot=<-0.1337 -0.2594 -0.0914 -0.9521>}
 dyn_hair_back_l_03 {parent=dyn_hair_back_l_02, loc=<3.0791 1.8093 0.9682>, rot=<-0.21 0.2362 -0.0524 -0.9473>}
 dyn_hair_back_l_04 {parent=dyn_hair_back_l_03, loc=<5.2678 0.3899 0.7156>, rot=<-0.0872 0 0 -0.9962>}
 dyn_hair_back_l_06 {parent=dyn_hair_back_l_04, loc=<7.5412 1.1145 2.7785>, rot=<0.1361 0.2929 0.1473 -0.9349>}
 dyn_hair_back_l_05 {parent=dyn_hair_back_l_04, loc=<4.6521 0.486 1.1305>, rot=<-0.0324 -0.2945 0.2485 -0.9222>}
 dyn_hair_back_l_07 {parent=dyn_hair_back_l_04, loc=<2.6796 -2.2864 0.6881>, rot=<0.04 0.0966 0.3806 -0.9188>}
 dyn_hair_side_l_base {parent=head, loc=<2.2249 1.9643 -7.6992>, rot=<0.0091 -0.1734 0.9835 -0.0515>}
 dyn_hair_side_l_parent {parent=dyn_hair_side_l_base, loc=<0 0 0>, rot=<0 0 0 1>}
 dyn_hair_side_l_07 {parent=dyn_hair_side_l_parent, loc=<0.6163 -3.5304 -0.6736>, rot=<0.1677 -0.0664 0.2872 -0.9407>}
 dyn_hair_side_l_05 {parent=dyn_hair_side_l_parent, loc=<5.9002 2.6808 -0.7973>, rot=<0.0657 -0.1179 -0.685 -0.7159>}
 dyn_hair_side_l_04 {parent=dyn_hair_side_l_parent, loc=<2.8832 4.8107 -0.6176>, rot=<-0.0965 -0.059 -0.8065 -0.5803>}
 dyn_hair_side_l_01 {parent=dyn_hair_side_l_parent, loc=<5.291 -0.9571 0.8568>, rot=<0 0 0 1>}
 dyn_hair_side_l_02 {parent=dyn_hair_side_l_01, loc=<4.0234 1.0836 0.0666>, rot=<0 0 0 1>}
 dyn_hair_side_l_03 {parent=dyn_hair_side_l_02, loc=<1.2333 2.8378 -2.0966>, rot=<-0.1318 -0.5242 -0.1904 -0.8195>}
 dyn_hair_side_l_06 {parent=dyn_hair_side_l_parent, loc=<5.6315 -1.9971 -0.771>, rot=<0 0 0 1>}
 dyn_hair_front_r_base {parent=head, loc=<0.5346 7.4393 6.821>, rot=<-0.0171 -0.978 0.2079 0.0036>}
 dyn_hair_front_r_parent {parent=dyn_hair_front_r_base, loc=<0 0 0>, rot=<0 0 0 1>}
 dyn_hair_front_r_01 {parent=dyn_hair_front_r_parent, loc=<4.6702 -0.4052 1.2033>, rot=<0 0.1392 0 -0.9903>}
 dyn_hair_front_r_02 {parent=dyn_hair_front_r_01, loc=<4.5846 -0.0567 -0.3233>, rot=<-0.1736 0 0 -0.9848>}
 dyn_hair_front_r_03 {parent=dyn_hair_front_r_02, loc=<3.6874 -1.317 0.3788>, rot=<-0.0818 -0.3285 -0.0152 -0.9408>}
 dyn_hair_front_r_04 {parent=dyn_hair_front_r_03, loc=<5.2689 2.2587 -0.6207>, rot=<-0.2585 0.3078 -0.4266 -0.8102>}
 dyn_hair_front_r_05 {parent=dyn_hair_front_r_01, loc=<3.6188 -1.6992 -3.8275>, rot=<-0.2611 -0.3946 -0.0479 -0.8797>}
 dyn_hair_leaf_02 {parent=head, loc=<4.2314 2.828 9.1707>, rot=<0.8715 -0.4441 0.1853 0.0944>}
 dyn_hair_leaf_01 {parent=head, loc=<10.7547 5.0017 7.7473>, rot=<0.8452 -0.4192 -0.0775 0.3222>}
 dyn_hair_bang_base {parent=head, loc=<11.6568 9.849 -0.9329>, rot=<0.2994 0.3185 -0.3978 -0.8067>}
 dyn_hair_bang_parent {parent=dyn_hair_bang_base, loc=<0 0 0>, rot=<0 0 0 1>}
 dyn_hair_bang_01 {parent=dyn_hair_bang_parent, loc=<1.7033 -0.1008 0.3388>, rot=<0 0.682 0 -0.7314>}
 dyn_hair_bang_02 {parent=dyn_hair_bang_01, loc=<4.8843 0.9144 2.1918>, rot=<0.5263 0.1199 0.8295 0.1436>}
 clavicle_l {parent=spine_05, loc=<5.4831 -0.1053 -3.2946>, rot=<0.0938 -0.7388 0.0307 -0.6666>}
 upperarm_l {parent=clavicle_l, loc=<9.9486 0 0>, rot=<0.0432 -0.2973 -0.0725 -0.951>}
 upperarm_twist_01_l {parent=upperarm_l, loc=<7.2131 0.1149 -0.0692>, rot=<-0.0062 -0.0048 -0.0158 -0.9998>}
 upperarm_twist_02_l {parent=upperarm_l, loc=<19.7925 0.513 -0.1874>, rot=<-0.0062 -0.0048 -0.0158 -0.9998>}
 clavicle_r {parent=spine_05, loc=<5.4831 -0.1053 3.2946>, rot=<0.7388 0.0938 -0.6666 -0.0307>}
 upperarm_r {parent=clavicle_r, loc=<-9.9486 0 0>, rot=<0.0432 -0.2973 -0.0725 -0.951>}
 upperarm_twist_01_r {parent=upperarm_r, loc=<-7.2131 -0.1149 0.0692>, rot=<-0.0062 -0.0048 -0.0158 -0.9998>}
 upperarm_twist_02_r {parent=upperarm_r, loc=<-19.7925 -0.513 0.1874>, rot=<-0.0062 -0.0048 -0.0158 -0.9998>}
]
variables [
 Crouch {type=double, default=0}
 Crouch_Move {type=double, default=0}
 head_SPR_look_down_alpha {type=double, default=0}
 head_SPR_look_left_alpha {type=double, default=0}
 head_SPR_look_right_alpha {type=double, default=0}
 head_SPR_look_up_alpha {type=double, default=0}
 look_down_upperarm_r_back_multiplied_alpha {type=double, default=0}
 look_up_and_left_two_hands_multiplied_alpha {type=double, default=0}
 look_up_two_hands_rifle_multiplied_alpha {type=double, default=0}
 uppearm_r_SPR_aim_up_two_hands_alpha {type=double, default=0}
 upperarm_r_SPR_back_alpha {type=double, default=0}
 upperarm_r_rifle_up_alpha {type=double, default=0}
 upperarm_r_skydiving {type=double, default=0}
]
