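# Physics asset for the F_MED_Parcel_Petal female character body, set up for dynamic hair on the bang, side, back and front-right strands.
# Box bodies form constrained chains from base bones through parent bones, some branching, with tight twist limits and per-strand swing limits.
# Middle and right back chains mostly favor one swing axis.
# Capsules on head, neck, upper spine, clavicles and upper arms supply body shapes, with only the left arm-to-clavicle joint constrained, rotationally free.
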PhysicsAsset /BRCosmetics/Characters/Player/Female/Medium/Bodies/F_MED_Parcel_Petal/Meshes/Parts/F_MED_Parcel_Petal_Physics
bodies {
 clavicle_l [{shapeType=Capsule, length=15.7455, radius=7.3377, center=<6.8424 6.417 -7.5071>, rotation=<rot 12.8441 -0.2504 -50.7486>}]
 clavicle_r [{shapeType=Capsule, length=12.7455, radius=7.3377, center=<-4.9791 -7.9988 6.4186>, rotation=<rot 0.9523 20.1326 -50.6967>}]
 dyn_hair_back_l_01 [{shapeType=Box, extentX=4.3837, extentY=5.467, extentZ=2.3057, center=<2.1795 0 0>}]
 dyn_hair_back_l_02 [{shapeType=Box, extentX=3.2189, extentY=5.6339, extentZ=2.1894, center=<1.5836 0 0>}]
 dyn_hair_back_l_03 [{shapeType=Box, extentX=4.5251, extentY=6.8544, extentZ=2.2394, center=<2.3458 0 0>}]
 dyn_hair_back_l_04 [{shapeType=Box, extentX=4.2841, extentY=6.8413, extentZ=2.7175, center=<1.7916 0 0>}]
 dyn_hair_back_l_05 [{shapeType=Box, extentX=3.6986, extentY=2.3412, extentZ=2.3038, center=<1.8389 0 0>}]
 dyn_hair_back_l_06 [{shapeType=Box, extentX=3.8873, extentY=2.0931, extentZ=2.2139, center=<2.0015 0 0>}]
 dyn_hair_back_l_07 [{shapeType=Box, extentX=3.2189, extentY=5.2759, extentZ=1.7064, center=<1.5836 -0.8333 0>, rotation=<rot -5.4687 0 0>}]
 dyn_hair_back_l_base [{shapeType=Box, extentX=0.5001, extentY=0.5001, extentZ=0.5001}]
 dyn_hair_back_l_parent [{shapeType=Box, extentX=5.1476, extentY=4.9878, extentZ=1.9963, center=<2.7025 0 0>}]
 dyn_hair_back_m_01 [{shapeType=Box, extentX=4.2905, extentY=5.5769, extentZ=2.3672, center=<2.4386 0 0>}]
 dyn_hair_back_m_02 [{shapeType=Box, extentX=4.2905, extentY=6.7948, extentZ=2.1803, center=<2.4386 0 0>}]
 dyn_hair_back_m_03 [{shapeType=Box, extentX=5.0233, extentY=6.6045, extentZ=2.4379, center=<2.6177 0 0>}]
 dyn_hair_back_m_04 [{shapeType=Box, extentX=5.8993, extentY=5.5769, extentZ=2.7151, center=<2.6846 0 -0.9649>, rotation=<rot 0 0 11.2762>}]
 dyn_hair_back_m_05 [{shapeType=Box, extentX=4.2905, extentY=5.9584, extentZ=3.0314, center=<2.4386 0 -1.6475>, rotation=<rot 0 0 10.4511>}]
 dyn_hair_back_m_06 [{shapeType=Box, extentX=2.3485, extentY=2.1577, extentZ=1.7527, center=<1.0906 0 -0.0001>}]
 dyn_hair_back_m_07 [{shapeType=Box, extentX=4.3457, extentY=4.0042, extentZ=2.692, center=<1.4989 0 -1.0485>, rotation=<rot 19.2783 0 0>}]
 dyn_hair_back_m_base [{shapeType=Box, extentX=0.5001, extentY=0.5001, extentZ=0.5001}]
 dyn_hair_back_m_parent [{shapeType=Box, extentX=1.01, extentY=1.01, extentZ=0.5}]
 dyn_hair_back_r_01 [{shapeType=Box, extentX=4.9274, extentY=6.5688, extentZ=2.6762, center=<2.6724 0 0>, rotation=<rot 5.2305 0 0>}]
 dyn_hair_back_r_02 [{shapeType=Box, extentX=4.3108, extentY=8.6348, extentZ=2.6703, center=<2.4949 0 0>, rotation=<rot 0 0 8.6179>}]
 dyn_hair_back_r_03 [{shapeType=Box, extentX=4.9738, extentY=8.1457, extentZ=2.8345, center=<2.9942 0 0>}]
 dyn_hair_back_r_04 [{shapeType=Box, extentX=4.7825, extentY=4.5451, extentZ=3.0092, center=<2.3136 0 0>}]
 dyn_hair_back_r_05 [{shapeType=Box, extentX=8.1389, extentY=3.9873, extentZ=2.2926, center=<4.1675 0.0001 -0.1997>, rotation=<rot -14.8754 0 0>}]
 dyn_hair_back_r_06 [{shapeType=Box, extentX=3.6724, extentY=3.9505, extentZ=1.756, center=<1.7095 0 0>}]
 dyn_hair_back_r_base [{shapeType=Box, extentX=0.5001, extentY=0.5001, extentZ=0.5001}]
 dyn_hair_back_r_parent [{shapeType=Box, extentX=5.4397, extentY=6.2711, extentZ=2.176, center=<2.6535 0 0>}]
 dyn_hair_bang_01 [{shapeType=Box, extentX=3.6571, extentY=1.6346, extentZ=1.5788, center=<1.8851 0 0>}]
 dyn_hair_bang_02 [{shapeType=Box, extentX=2.9815, extentY=1.1919, extentZ=2.8692, center=<1.0859 -0.0001 0>}]
 dyn_hair_bang_base [{shapeType=Box, extentX=0.5001, extentY=0.5001, extentZ=0.5001}]
 dyn_hair_bang_l_base [{shapeType=Box, extentX=0.5001, extentY=0.5001, extentZ=0.5001}]
 dyn_hair_bang_l_parent [{shapeType=Box, extentX=1.159, extentY=3.2995, extentZ=3.0777, center=<0.5606 -1.352 0>, rotation=<rot 0 20.6275 0>}]
 dyn_hair_bang_parent [{shapeType=Box, extentX=2.8993, extentY=1.7516, extentZ=1.3436, center=<1.4229 0.0001 0>}]
 dyn_hair_front_r_01 [{shapeType=Box, extentX=4.4297, extentY=4.5563, extentZ=2.487, center=<2.1333 0 0>}]
 dyn_hair_front_r_02 [{shapeType=Box, extentX=4.4297, extentY=4.5563, extentZ=2.487, center=<2.1333 0 0>}]
 dyn_hair_front_r_03 [{shapeType=Box, extentX=4.4297, extentY=3.5563, extentZ=1.9391, center=<2.1333 0 0>}]
 dyn_hair_front_r_04 [{shapeType=Box, extentX=3.3421, extentY=2.0225, extentZ=1.8421, center=<1.631 -0.1089 0>, rotation=<rot 0 0 -74.2219>}]
 dyn_hair_front_r_05 [{shapeType=Box, extentX=2.9251, extentY=1.7098, extentZ=1.781, center=<1.3935 0 0>}]
 dyn_hair_front_r_base [{shapeType=Box, extentX=0.5001, extentY=0.5001, extentZ=0.5001}]
 dyn_hair_front_r_parent [{shapeType=Box, extentX=4.4297, extentY=4.5563, extentZ=2.487, center=<2.1333 0 0>}]
 dyn_hair_side_l_01 [{shapeType=Box, extentX=3.7994, extentY=4.7004, extentZ=1.9027, center=<1.8315 0 0>}]
 dyn_hair_side_l_02 [{shapeType=Box, extentX=2.5649, extentY=4.2469, extentZ=1.613, center=<1.3479 0 0>}]
 dyn_hair_side_l_03 [{shapeType=Box, extentX=4.4494, extentY=4.2405, extentZ=2.9614, center=<2.4788 -0.0001 0.2922>, rotation=<rot -2.3935 0 0>}]
 dyn_hair_side_l_04 [{shapeType=Box, extentX=3.5764, extentY=2.8926, extentZ=0.9571, center=<1.8321 0 0>}]
 dyn_hair_side_l_05 [{shapeType=Box, extentX=2.3369, extentY=1.972, extentZ=1.0939, center=<1.0941 0 0>}]
 dyn_hair_side_l_06 [{shapeType=Box, extentX=3.7187, extentY=1.5717, extentZ=1.032, center=<1.7156 0 0>}]
 dyn_hair_side_l_07 [{shapeType=Box, extentX=3.3714, extentY=2.0566, extentZ=1.2292, center=<1.6964 0 0.0001>}]
 dyn_hair_side_l_base [{shapeType=Box, extentX=0.5001, extentY=0.5001, extentZ=0.5001}]
 dyn_hair_side_l_parent [{shapeType=Box, extentX=3.9531, extentY=4.6435, extentZ=1.6579, center=<1.8731 0 0>}]
 dyn_hair_side_r_01 [{shapeType=Box, extentX=4.1029, extentY=6.507, extentZ=1.061, center=<2.046 0 0>}]
 dyn_hair_side_r_02 [{shapeType=Box, extentX=3.6311, extentY=5.3262, extentZ=1.8342, center=<1.9618 0 0>}]
 dyn_hair_side_r_03 [{shapeType=Box, extentX=2.5105, extentY=5.7249, extentZ=1.5882, center=<1.1689 0 0>}]
 dyn_hair_side_r_base [{shapeType=Box, extentX=0.5001, extentY=0.5001, extentZ=0.5001}]
 dyn_hair_side_r_parent [{shapeType=Box, extentX=3.7742, extentY=6.0113, extentZ=1.7611, center=<1.6471 0 0.0001>}]
 head [{shapeType=Capsule, length=7.2643, radius=7.3727, center=<2.0778 0.654 0>, rotation=<rot 90 0 -14.9586>}]
 neck_02 [{shapeType=Capsule, length=4.4399, radius=6.9765, center=<-4.3132 1.6881 0>, rotation=<rot 90 0 0>}]
 spine_05 [
  {shapeType=Capsule, length=10.264, radius=9.2528, center=<-3.662 -1.222 0>, rotation=<rot 90 0 0>}
  {shapeType=Capsule, length=11.2535, radius=5.5078, center=<5.5477 -2.8152 0>}
 ]
 upperarm_l [{shapeType=Capsule, length=15.0592, radius=5.2366, shapeName=upperarm_l_capsule, center=<6 1 0>, rotation=<rot 0 90 90>}]
 upperarm_r [{shapeType=Capsule, length=24.2101, radius=5.7899, center=<-11.4766 0 -0.6424>, rotation=<rot 88.2431 -179.9999 -179.9999>}]
}
constraints [
 {bone1Name=dyn_hair_bang_l_parent, bone2Name=dyn_hair_bang_l_base, swing1LimitDegrees=20, swing2LimitDegrees=20, twistLimitDegrees=2}
 {bone1Name=dyn_hair_bang_parent, bone2Name=dyn_hair_bang_base, swing1LimitDegrees=5, swing2LimitDegrees=3, twistLimitDegrees=3}
 {bone1Name=dyn_hair_bang_01, bone2Name=dyn_hair_bang_parent, swing1LimitDegrees=3, swing2LimitDegrees=3, twistLimitDegrees=3}
 {bone1Name=dyn_hair_bang_02, bone2Name=dyn_hair_bang_01, swing1LimitDegrees=3, swing2LimitDegrees=3, twistLimitDegrees=2}
 {bone1Name=dyn_hair_side_l_parent, bone2Name=dyn_hair_side_l_base, swing1LimitDegrees=15, swing2LimitDegrees=15, twistLimitDegrees=3}
 {bone1Name=dyn_hair_side_l_07, bone2Name=dyn_hair_side_l_parent, swing1LimitDegrees=8, swing2LimitDegrees=8, twistLimitDegrees=2}
 {bone1Name=dyn_hair_side_l_05, bone2Name=dyn_hair_side_l_parent, swing1LimitDegrees=8, swing2LimitDegrees=8, twistLimitDegrees=3}
 {bone1Name=dyn_hair_side_l_04, bone2Name=dyn_hair_side_l_parent, swing1LimitDegrees=10, swing2LimitDegrees=15, twistLimitDegrees=3}
 {bone1Name=dyn_hair_side_l_01, bone2Name=dyn_hair_side_l_parent, swing1LimitDegrees=10, swing2LimitDegrees=10, twistLimitDegrees=3}
 {bone1Name=dyn_hair_side_l_02, bone2Name=dyn_hair_side_l_01, swing1LimitDegrees=10, swing2LimitDegrees=10, twistLimitDegrees=3}
 {bone1Name=dyn_hair_side_l_03, bone2Name=dyn_hair_side_l_02, swing1LimitDegrees=10, swing2LimitDegrees=10, twistLimitDegrees=5}
 {bone1Name=dyn_hair_back_l_parent, bone2Name=dyn_hair_back_l_base, swing1LimitDegrees=10, swing2LimitDegrees=10, twistLimitDegrees=3}
 {bone1Name=dyn_hair_back_l_01, bone2Name=dyn_hair_back_l_parent, swing1LimitDegrees=10, swing2LimitDegrees=10, twistLimitDegrees=3}
 {bone1Name=dyn_hair_back_l_02, bone2Name=dyn_hair_back_l_01, swing1LimitDegrees=10, swing2LimitDegrees=10, twistLimitDegrees=3}
 {bone1Name=dyn_hair_back_l_03, bone2Name=dyn_hair_back_l_02, swing1LimitDegrees=10, swing2LimitDegrees=10, twistLimitDegrees=3}
 {bone1Name=dyn_hair_back_l_04, bone2Name=dyn_hair_back_l_03, swing1LimitDegrees=10, swing2LimitDegrees=10, twistLimitDegrees=5}
 {bone1Name=dyn_hair_back_l_06, bone2Name=dyn_hair_back_l_04, swing1LimitDegrees=10, swing2LimitDegrees=10, twistLimitDegrees=3}
 {bone1Name=dyn_hair_back_l_05, bone2Name=dyn_hair_back_l_04, swing1LimitDegrees=10, swing2LimitDegrees=10, twistLimitDegrees=3}
 {bone1Name=dyn_hair_back_l_07, bone2Name=dyn_hair_back_l_04, swing1LimitDegrees=10, swing2LimitDegrees=10, twistLimitDegrees=3}
 {bone1Name=dyn_hair_back_m_parent, bone2Name=dyn_hair_back_m_base, swing1LimitDegrees=3, swing2LimitDegrees=10, twistLimitDegrees=3}
 {bone1Name=dyn_hair_back_m_01, bone2Name=dyn_hair_back_m_parent, swing1LimitDegrees=3, swing2LimitDegrees=30, twistLimitDegrees=5}
 {bone1Name=dyn_hair_back_m_02, bone2Name=dyn_hair_back_m_01, swing1LimitDegrees=3, swing2LimitDegrees=20, twistLimitDegrees=8}
 {bone1Name=dyn_hair_back_m_03, bone2Name=dyn_hair_back_m_02, swing1LimitDegrees=3, swing2LimitDegrees=20, twistLimitDegrees=8}
 {bone1Name=dyn_hair_back_m_04, bone2Name=dyn_hair_back_m_03, swing1LimitDegrees=3, swing2LimitDegrees=10, twistLimitDegrees=8}
 {bone1Name=dyn_hair_back_m_05, bone2Name=dyn_hair_back_m_04, swing1LimitDegrees=3, swing2LimitDegrees=10, twistLimitDegrees=8}
 {bone1Name=dyn_hair_back_m_06, bone2Name=dyn_hair_back_m_05, swing1LimitDegrees=3, swing2LimitDegrees=10, twistLimitDegrees=5}
 {bone1Name=dyn_hair_back_m_07, bone2Name=dyn_hair_back_m_05, swing1LimitDegrees=3, swing2LimitDegrees=10, twistLimitDegrees=5}
 {bone1Name=dyn_hair_back_r_parent, bone2Name=dyn_hair_back_r_base, swing1LimitDegrees=3, swing2LimitDegrees=40, twistLimitDegrees=3}
 {bone1Name=dyn_hair_back_r_01, bone2Name=dyn_hair_back_r_parent, swing1LimitDegrees=3, swing2LimitDegrees=35, twistLimitDegrees=5}
 {bone1Name=dyn_hair_back_r_02, bone2Name=dyn_hair_back_r_01, swing1LimitDegrees=3, swing2LimitDegrees=45, twistLimitDegrees=5}
 {bone1Name=dyn_hair_back_r_03, bone2Name=dyn_hair_back_r_02, swing1LimitDegrees=3, swing2LimitDegrees=45, twistLimitDegrees=5}
 {bone1Name=dyn_hair_back_r_04, bone2Name=dyn_hair_back_r_03, swing1LimitDegrees=3, swing2LimitDegrees=20, twistLimitDegrees=3}
 {bone1Name=dyn_hair_back_r_06, bone2Name=dyn_hair_back_r_04, swing1LimitDegrees=20, swing2LimitDegrees=20, twistLimitDegrees=3}
 {bone1Name=dyn_hair_back_r_05, bone2Name=dyn_hair_back_r_04, swing1LimitDegrees=20, swing2LimitDegrees=20, twistLimitDegrees=5}
 {bone1Name=dyn_hair_side_r_parent, bone2Name=dyn_hair_side_r_base, swing1LimitDegrees=25, swing2LimitDegrees=20, twistLimitDegrees=5}
 {bone1Name=dyn_hair_side_r_01, bone2Name=dyn_hair_side_r_parent, swing1LimitDegrees=25, swing2LimitDegrees=20, twistLimitDegrees=5}
 {bone1Name=dyn_hair_side_r_02, bone2Name=dyn_hair_side_r_01, swing1LimitDegrees=25, swing2LimitDegrees=20, twistLimitDegrees=5}
 {bone1Name=dyn_hair_side_r_03, bone2Name=dyn_hair_side_r_02, swing1LimitDegrees=25, swing2LimitDegrees=20, twistLimitDegrees=5}
 {bone1Name=dyn_hair_front_r_parent, bone2Name=dyn_hair_front_r_base, swing1LimitDegrees=25, swing2LimitDegrees=25, twistLimitDegrees=3}
 {bone1Name=dyn_hair_front_r_01, bone2Name=dyn_hair_front_r_parent, swing1LimitDegrees=25, swing2LimitDegrees=25, twistLimitDegrees=3}
 {bone1Name=dyn_hair_front_r_02, bone2Name=dyn_hair_front_r_01, swing1LimitDegrees=25, swing2LimitDegrees=25, twistLimitDegrees=3}
 {bone1Name=dyn_hair_front_r_03, bone2Name=dyn_hair_front_r_02, swing1LimitDegrees=25, swing2LimitDegrees=25, twistLimitDegrees=3}
 {bone1Name=dyn_hair_front_r_04, bone2Name=dyn_hair_front_r_03, swing1LimitDegrees=30, swing2LimitDegrees=30, twistLimitDegrees=3}
 {bone1Name=dyn_hair_front_r_05, bone2Name=dyn_hair_front_r_01, swing1LimitDegrees=25, swing2LimitDegrees=25, twistLimitDegrees=3}
 {bone1Name=dyn_hair_side_l_06, bone2Name=dyn_hair_side_l_parent, swing1LimitDegrees=20, swing2LimitDegrees=20, twistLimitDegrees=3}
 {
  bone1Name=upperarm_l
  bone2Name=clavicle_l
  swing1LimitDegrees=45
  swing1Motion=Free
  swing2LimitDegrees=45
  swing2Motion=Free
  twistLimitDegrees=45
  twistMotion=Free
 }
]
