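# Animation Blueprint for the Parcel Petal cosmetic part that drives secondary motion of its hair and leaves.
# Copies the parent pose, constrains back hair between spine and head, then runs a crouch-driven Control Rig, RigidBody and leaf AnimDynamics.
# EventGraph reads head-to-spine rotation, targeting, crouch movement and an idle curve.
# These values smoothly rotate right-side hair while aiming with head turned, and offset hair bases during front-end idle.
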AnimBlueprint /BRCosmetics/Characters/Player/Female/Medium/Bodies/F_MED_Parcel_Petal/Meshes/Parts/F_MED_Parcel_Petal_AnimBP
target_skeleton=&/F_MED_Parcel_Petal_Skeleton
graphs [
 AnimGraph {nodes [
   n1 = Get ADS_Rotate() -> (ADS_Rotate:double)
   "Transform (Modify) Bone - Bone: dyn_hair_back_r_base"(Alpha=@n1, Translation="0,0,0", Rotation="0,0,-10", Scale="1,1,1") {
    rotationMode=BMM_Additive
    rotationSpace=BCS_BoneSpace
    scaleSpace=BCS_BoneSpace
    translationMode=BMM_Additive
    translationSpace=BCS_BoneSpace
    alphaBoolBlend {blendOption=Linear}
    alphaScaleBiasClamp {bInterpResult=true, interpSpeedDecreasing=4, interpSpeedIncreasing=6}
    boneToModify {boneName=dyn_hair_back_r_base}
    rotation {roll=-10}
   }
   n2 = "Constraint - dyn_hair_back_l_base (PP)"!(
    ComponentPose=Local To Component(LocalPose=Copy Pose From Mesh() {bUseAttachedParent=true})
    Alpha=1
    ConstraintWeights_0=0.6
    ConstraintWeights_1:float=0.4
   ) -> (Pose) {
    alphaBoolBlend {blendOption=Linear}
    boneToModify {boneName=dyn_hair_back_l_base}
    constraintSetup [{targetBone {boneName=spine_05}}, {targetBone {boneName=head}}]
    constraintWeights [1,1]
   }
   n3 = "Constraint - dyn_hair_back_r_base (PP)"!(ComponentPose=@n2, Alpha=1, ConstraintWeights_0=0.6, ConstraintWeights_1:float=0.4) -> (Pose) {
    alphaBoolBlend {blendOption=Linear}
    boneToModify {boneName=dyn_hair_back_r_base}
    constraintSetup [{targetBone {boneName=spine_05}}, {targetBone {boneName=head}}]
    constraintWeights [1,1]
   }
   n4 = "Constraint - dyn_hair_back_m_base (PP)"!(ComponentPose=@n3, Alpha=1, ConstraintWeights_0=0.6, ConstraintWeights_1:float=0.4) -> (Pose) {
    alphaBoolBlend {blendOption=Linear}
    boneToModify {boneName=dyn_hair_back_m_base}
    constraintSetup [{targetBone {boneName=spine_05}}, {targetBone {boneName=head}}]
    constraintWeights [1,1]
   }
   n5 = Control Rig!(
    Source=Component To Local(ComponentPose=@n4)
    Alpha
    Crouch:double=Get Crouch() -> (Crouch:double)
    Crouch_Move:double=Get Crouch_Move() -> (Crouch_Move:double)
   ) -> (Pose) {
    alphaBoolBlend {blendOption=Linear}
    controlRigAssetReference {blueprintRigClass=&/F_MED_Parcel_Petal_CtrlRig1.F_MED_Parcel_Petal_CtrlRig1_C}
   }
   n6 = RigidBody(ComponentPose=Local To Component(LocalPose=@n5)) {
    bClampLinearTranslationLimitToRefPose=true
    bOverrideWorldGravity=true
    overridePhysicsAsset=&/F_MED_Parcel_Petal_Physics
    simulationSpace=BaseBoneSpace
    alphaBoolBlend {blendOption=Linear}
    baseBoneRef {boneName=spine_05}
    componentAppliedLinearAccClamp=<6000 6000 6000>
    componentLinearAccScale=<0.15 0.15 0.45>
    componentLinearVelScale=<0.15 0.15 0.4>
    overrideWorldGravity {z=-980}
   }
   n7 = "Transform (Modify) Bone - Bone: dyn_hair_side_r_base"(ComponentPose=@n6, Alpha=@n1, Translation="0,0,0", Rotation="30,0,0.000", Scale="1,1,1") {
    rotationMode=BMM_Additive
    rotationSpace=BCS_BoneSpace
    scaleSpace=BCS_BoneSpace
    translationMode=BMM_Additive
    translationSpace=BCS_BoneSpace
    alphaBoolBlend {blendOption=Linear}
    alphaScaleBiasClamp {bInterpResult=true, interpSpeedDecreasing=4, interpSpeedIncreasing=6}
    boneToModify {boneName=dyn_hair_side_r_base}
    rotation {pitch=30}
   }
   n8 = "Transform (Modify) Bone - Bone: dyn_hair_front_r_base"(ComponentPose=@n7, Alpha=@n1, Translation="0,0,0", Rotation="30,0,0.000", Scale="1,1,1") {
    rotationMode=BMM_Additive
    rotationSpace=BCS_BoneSpace
    scaleMode=BMM_Replace
    scaleSpace=BCS_BoneSpace
    translationMode=BMM_Additive
    translationSpace=BCS_BoneSpace
    alphaBoolBlend {blendOption=Linear}
    alphaScaleBiasClamp {bInterpResult=true, interpSpeedDecreasing=4, interpSpeedIncreasing=6}
    boneToModify {boneName=dyn_hair_front_r_base}
    rotation {pitch=30}
   }
   n9 = Anim Dynamics - Bone: dyn_hair_leaf_01(ComponentPose=@n8, Alpha=0.75, ExternalForce="0,0,0") {
    angularDampingOverride=0.999
    bGravityOverrideInSimSpace=true
    bOverrideAngularDamping=true
    bOverrideLinearDamping=true
    bUseGravityOverride=true
    linearDampingOverride=0.999
    simulationSpace=RootRelative
    alphaBoolBlend {blendOption=Linear}
    boundBone {boneName=dyn_hair_leaf_01}
    componentAppliedLinearAccClamp=<3000 3000 3000>
    componentLinearAccScale=<0.25 0.25 0.5>
    componentLinearVelScale=<0.25 0.25 0.5>
    gravityOverride=<0.85 0.85 0.85>
    physicsBodyDefinitions [{
      sphereCollisionRadius=0
      boundBone {boneName=dyn_hair_leaf_01}
      boxExtents=<6 4 2>
      constraintSetup {angularLimitsMax=<0 20 10>, angularLimitsMin=<0 -10 -10>, angularTarget=<0 0 0>}
      localJointOffset=<2 0 0>
     }]
    relativeSpaceBone {boneName=spine_05}
   }
   n10 = Anim Dynamics - Bone: dyn_hair_leaf_02(ComponentPose=@n9, Alpha=0.75, ExternalForce="0,0,0") {
    angularDampingOverride=0.999
    bGravityOverrideInSimSpace=true
    bOverrideAngularDamping=true
    bOverrideLinearDamping=true
    bUseGravityOverride=true
    linearDampingOverride=0.999
    simulationSpace=RootRelative
    alphaBoolBlend {blendOption=Linear}
    boundBone {boneName=dyn_hair_leaf_02}
    componentAppliedLinearAccClamp=<3000 3000 3000>
    componentLinearAccScale=<0.25 0.25 0.5>
    componentLinearVelScale=<0.25 0.25 0.5>
    gravityOverride=<0.85 0.85 0.85>
    physicsBodyDefinitions [{
      sphereCollisionRadius=0
      boundBone {boneName=dyn_hair_leaf_02}
      boxExtents=<6 4 2>
      constraintSetup {angularLimitsMax=<0 20 10>, angularLimitsMin=<0 -1 -10>, angularTarget=<0 0 0>}
      localJointOffset=<2 0 0>
     }]
    relativeSpaceBone {boneName=spine_05}
   }
   n11 = Anim Dynamics - Bone: dyn_hair_leaf_03(ComponentPose=@n10, Alpha=0.75, ExternalForce="0,0,0") {
    angularDampingOverride=0.999
    bGravityOverrideInSimSpace=true
    bOverrideAngularDamping=true
    bOverrideLinearDamping=true
    bUseGravityOverride=true
    linearDampingOverride=0.999
    simulationSpace=RootRelative
    alphaBoolBlend {blendOption=Linear}
    boundBone {boneName=dyn_hair_leaf_03}
    componentAppliedLinearAccClamp=<3000 3000 3000>
    componentLinearAccScale=<0.25 0.25 0.5>
    componentLinearVelScale=<0.25 0.25 0.5>
    gravityOverride=<0.85 0.85 0.85>
    physicsBodyDefinitions [{
      sphereCollisionRadius=0
      boundBone {boneName=dyn_hair_leaf_03}
      boxExtents=<6 4 2>
      constraintSetup {angularLimitsMax=<0 20 10>, angularLimitsMin=<0 -10 -10>, angularTarget=<0 0 0>}
      localJointOffset=<2 0 0>
     }]
    relativeSpaceBone {boneName=spine_05}
   }
   n12 = Anim Dynamics - Bone: dyn_hair_leaf_04(ComponentPose=@n11, Alpha=0.75, ExternalForce="0,0,0") {
    angularDampingOverride=0.999
    bGravityOverrideInSimSpace=true
    bOverrideAngularDamping=true
    bOverrideLinearDamping=true
    bUseGravityOverride=true
    linearDampingOverride=0.999
    simulationSpace=RootRelative
    alphaBoolBlend {blendOption=Linear}
    boundBone {boneName=dyn_hair_leaf_04}
    componentAppliedLinearAccClamp=<3000 3000 3000>
    componentLinearAccScale=<0.25 0.25 0.5>
    componentLinearVelScale=<0.25 0.25 0.5>
    gravityOverride=<0.85 0.85 0.85>
    physicsBodyDefinitions [{
      sphereCollisionRadius=0
      boundBone {boneName=dyn_hair_leaf_04}
      boxExtents=<6 4 2>
      constraintSetup {angularLimitsMax=<0 20 10>, angularLimitsMin=<0 -10 -10>, angularTarget=<0 0 0>}
      localJointOffset=<2 0 0>
     }]
    relativeSpaceBone {boneName=spine_05}
   }
   n13 = Get IsPlayingIdle() -> (IsPlayingIdle:bool)
   n14 = "Transform (Modify) Bone - Bone: dyn_hair_back_l_base"!(ComponentPose=@n12, bAlphaBoolEnabled:bool=@n13, Translation="0,2,0", Rotation="0,0,0", Scale="1,1,1") -> (Pose) {
    alphaInputType=Bool
    rotationMode=BMM_Additive
    rotationSpace=BCS_BoneSpace
    translationMode=BMM_Additive
    translationSpace=BCS_BoneSpace
    alphaBoolBlend {blendOption=Linear}
    boneToModify {boneName=dyn_hair_back_l_base}
    translation {y=2}
   }
   n15 = "Transform (Modify) Bone - Bone: dyn_hair_back_m_base"!(ComponentPose=@n14, bAlphaBoolEnabled:bool=@n13, Translation="0,2,1", Rotation="-3.725,0.000,0.000", Scale="1,1,1") -> (Pose) {
    alphaInputType=Bool
    rotationMode=BMM_Additive
    rotationSpace=BCS_BoneSpace
    translationMode=BMM_Additive
    translationSpace=BCS_BoneSpace
    alphaBoolBlend {blendOption=Linear}
    boneToModify {boneName=dyn_hair_back_m_base}
    rotation {pitch=-3.725}
    translation {y=2, z=1}
   }
   n16 = "Transform (Modify) Bone - Bone: dyn_hair_front_r_base"!(ComponentPose=@n15, bAlphaBoolEnabled:bool=@n13, Translation="0,2,-2", Rotation="0,0,0", Scale="1,1,1") -> (Pose) {
    alphaInputType=Bool
    rotationMode=BMM_Additive
    rotationSpace=BCS_BoneSpace
    translationMode=BMM_Additive
    translationSpace=BCS_BoneSpace
    alphaBoolBlend {blendOption=Linear}
    boneToModify {boneName=dyn_hair_front_r_base}
    translation {y=2, z=-2}
   }
   Output Pose(Result=Component To Local(ComponentPose=@n16)) {layerGroup=None}
  ]}
 EventGraph {nodes [
   Event BlueprintUpdateAnimation()
   > Get!(execute:Exec) -> (then:Exec, else:Exec, Skel_Mesh:Skeletal Mesh Component Object Reference)
   > Get!(execute) -> (then, else, Anim_Inst:Anim Instance Object Reference)
   > n1 = Sequence!(execute) -> (then_0, then_1, then_2:Exec)
   n2 = BreakTransform(InTransform=GetDeltaTransformFromRefPose(self=GetOwnerSkeletalMeshComponent(), BoneName=head, BaseName=spine_05))
   > Set Head_Rot_Back(Head_Rot_Back:double=MapRangeClamped(Value=BreakRotator(InRot=@n2.Rotation).Yaw, InRangeA=0, InRangeB=-48)) -> (Output_Get:double)
   n3 = "In Range (Float)"!(Value:double=Get Head_Rot_Back() -> (Head_Rot_Back:double), Min:double=0.2, Max:double=1, InclusiveMin:bool=true, InclusiveMax:bool=true) -> (ReturnValue:bool)
   n4 = Get bIsTargeting()
   Set ADS_Rotate(execute=@n1.then_1, ADS_Rotate:double=SelectFloat(A=1, B=0, bPickA=AND Boolean(A=@n3, B=@n4))) -> (Output_Get:double)
   n5 = Get Anim_Inst() -> (Anim_Inst:Anim Instance Object Reference)
   Is Valid(exec=@n1.then_2, InputObject=@n5)
   > Set IsPlayingIdle(IsPlayingIdle:bool="float > float"(A=GetCurveValue(self=@n5, CurveName=FrontEnd_Idle_Bypass), B=0) {is=title K2Node_CallFunction}) -> (Output_Get:bool)
   n6 = Get bIsCrouching()
   > Set Crouch(Crouch:double="To Float (Boolean)"(InBool=@n6)) -> (Output_Get:double)
   > n7 = Branch(Condition=AND Boolean(A=@n6, B=Get bIsMoving2D()))
   > Set Crouch_Move(Crouch_Move:double=1) -> (Output_Get:double)
   Set Crouch_Move(execute=@n7.else, Crouch_Move=0)
   Event BlueprintInitializeAnimation()
   n8 = GetOwnerSkeletalMeshComponent()
   > Is Valid(InputObject=@n8)
   > Set Skel_Mesh(Skel_Mesh=@n8)
   > n9 = Cast To FortPlayerPawn(Object=GetOwningActor())
   > Is Valid(InputObject=@n9)
   > n10 = Set Pawn(Pawn:Fort Player Pawn Object Reference=@n9) -> (Output_Get:Fort Player Pawn Object Reference)
   n11 = GetAnimInstance(self=Get Mesh(self=@n10))
   > Is Valid(InputObject=@n11)
   > Set Anim_Inst(Anim_Inst:Anim Instance Object Reference=@n11) -> (Output_Get:Anim Instance Object Reference)
   Set MoveHair(MoveHair:bool=false) -> (Output_Get:bool)
   > Branch(Condition=Get MoveHair() -> (MoveHair:bool))
   > Set Alpha(Alpha:double=0) -> (Output_Get:double)
   "float > float"(A=GetCurveValue(CurveName=Disable_Custom), B=0)
  ]}
]
variables [
 ADS_Rotate {type=number}
 Alpha {type=number}
 Anim_Inst {type=AnimInstance, default=null}
 Body_Mesh {type=SkeletalMeshComponent}
 Crouch {type=number}
 Crouch_Move {type=number}
 Head_Rot_Back {type=number}
 IsPlayingIdle {type=boolean}
 MoveHair {type=boolean}
 Pawn {type=FortPlayerPawn}
 Pawn Owner {type=PlayerPawn_Athena_C}
 Skel_Mesh {type=SkeletalMeshComponent}
]
