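# Control Rig for the F_MED_Parcel_Petal character body that keeps the back hair bones following the upper body.
# Parent constraints blend left, middle and right hair bones and nulls across head, neck, spine, clavicles, upper arms and other hair bones.
# Spherical pose readers on both upper arms then weight corrective hair offsets, some scaled by BHairDyn, JumpFix or Crounchfix.
# A separate Controls graph stores the middle hair controls and their pose at setup.
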ControlRigBlueprint /BRCosmetics/Characters/Player/Female/Medium/Bodies/F_MED_Parcel_Petal/Meshes/Parts/F_MED_Parcel_Petal_CtrlRig
graphs [
 ContainedGraph {nodes [
   n1 = RigVMFunctionEntryNode()
   SequenceAggregate = &RigVMFunction_Sequence(ExecuteContext=@n1.ExecuteContext)
   SequenceAggregate_1 = &RigVMFunction_Sequence(ExecuteContext=@SequenceAggregate.B)
   SequenceAggregate_2 = &RigVMFunction_Sequence(ExecuteContext=@SequenceAggregate_1.B)
   RigVMFunctionReturnNode(A=@SequenceAggregate.A, B=@SequenceAggregate_1.A, C=@SequenceAggregate_2.A, D=@SequenceAggregate_2.B)
  ]}
 Multiply_1_1_ContainedGraph {nodes [
   n1 = RigVMFunctionEntryNode()
   RigVMFunctionReturnNode(Result=&RigVMFunction_MathDoubleMul(B=@n1.C, A=&RigVMFunction_MathDoubleMul(A=@n1.A, B=@n1.B).Result).Result)
  ]}
 Multiply_2_ContainedGraph {nodes [
   n1 = RigVMFunctionEntryNode()
   Multiply = &RigVMFunction_MathDoubleMul(A=@n1.A, B=@n1.B)
   RigVMFunctionReturnNode(Result=&RigVMFunction_MathDoubleMul(B=@n1.C, A=@Multiply.Result).Result)
  ]}
 RigVMFunctionLibrary {links [], nodes []}
 RigVMFunction_Sequence_2_ContainedGraph {nodes [
   n1 = RigVMFunctionEntryNode()
   n2 = &RigVMFunction_Sequence(ExecuteContext=@n1.ExecuteContext)
   n3 = &RigVMFunction_Sequence(ExecuteContext=@n2.B)
   RigVMFunctionReturnNode(A=@n2.A, B=@n3.A, C=@n3.B)
  ]}
 RigVMModel {nodes [
   RigUnit_BeginExecution = &RigUnit_BeginExecution()
   SequenceAggregate = &RigVMFunction_Sequence(ExecuteContext=@RigUnit_BeginExecution.ExecutePin)
   &RigUnit_ParentConstraint(
    AdvancedSettings=(RotationOrderForFilter=XZY)
    Child=(Type=Null,Name="dyn_hair_back_m_01_null")
    Filter=(TranslationFilter=(bX=true,bY=true,bZ=true),RotationFilter=(bX=false,bY=false,bZ=false),ScaleFilter=(bX=True,bY=True,bZ=True))
    Parents=((Item=(Type=Bone,Name="head"),Weight=0.9),(Item=(Type=Bone,Name="spine_05"),Weight=0.1))
    Weight=1
    ExecutePin=@SequenceAggregate.A
   )
   > &RigUnit_ParentConstraint(
    AdvancedSettings=(RotationOrderForFilter=XZY)
    Child=(Type=Null,Name="dyn_hair_back_m_02_null")
    Filter=(TranslationFilter=(bX=True,bY=True,bZ=True),RotationFilter=(bX=false,bY=false,bZ=false),ScaleFilter=(bX=True,bY=True,bZ=True))
    Parents=((Item=(Type=Bone,Name="neck_02"),Weight=0.6),(Item=(Type=Bone,Name="spine_05"),Weight=0.4))
    Weight=1
   )
   > &RigUnit_ParentConstraint(
    AdvancedSettings=(RotationOrderForFilter=XZY)
    Child=(Type=Null,Name="dyn_hair_back_l_03_null")
    Filter=(TranslationFilter=(bX=True,bY=True,bZ=True),RotationFilter=(bX=false,bY=false,bZ=false),ScaleFilter=(bX=True,bY=True,bZ=True))
    Parents=((Item=(Type=Bone,Name="neck_02"),Weight=0.5),(Item=(Type=Bone,Name="spine_05"),Weight=0.5))
    Weight=1
   )
   > &RigUnit_ParentConstraint(
    AdvancedSettings=(RotationOrderForFilter=XZY)
    Child=(Type=Bone,Name="dyn_hair_back_m_04")
    Parents=((Item=(Type=Bone,Name="neck_01"),Weight=0.0055),(Item=(Type=Bone,Name="spine_03"),Weight=0.0076),(Item=(Type=Bone,Name="clavicle_r"),Weight=0.0885),(Item=(Type=Bone,Name="clavicle_l"),Weight=0.0918),(Item=(Type=Bone,Name="spine_05"),Weight=0.8066))
    Weight=1
   )
   > &RigUnit_ParentConstraint(
    AdvancedSettings=(RotationOrderForFilter=XZY)
    Child=(Type=Bone,Name="dyn_hair_back_m_05")
    Parents="((Item=(Type=Bone,Name=\"spine_01\"),Weight=0.0025),(Item=(Type=Bone,Name=\"spine_03\"),Weight=0.0108),(Item=(Type=Bone,Name=\"spine_02\"),Weight=0.011),(Item=(Type=Bone,Name=\"upperarm_l\"),Weight=0.0192),(Item=(Type=Bone,Name=\"spine_04\"),Weight=0.1175),(Item=(Ty ...(truncated)"
    Weight=1
   )
   > &RigUnit_ParentConstraint(
    AdvancedSettings=(RotationOrderForFilter=XZY)
    Child=(Type=Bone,Name="dyn_hair_back_l_base")
    Filter=(TranslationFilter=(bX=true,bY=true,bZ=true),RotationFilter=(bX=true,bY=true,bZ=true),ScaleFilter=(bX=True,bY=True,bZ=True))
    Parents=((Item=(Type=Bone,Name="dyn_hair_back_l_parent"),Weight=1),(Item=(Type=Bone,Name="spine_05"),Weight=0))
    Weight=1
   )
   > &RigUnit_ParentConstraint(
    AdvancedSettings=(RotationOrderForFilter=XZY)
    Child=(Type=Bone,Name="dyn_hair_back_l_parent")
    Filter=(TranslationFilter=(bX=True,bY=True,bZ=True),RotationFilter=(bX=true,bY=true,bZ=true),ScaleFilter=(bX=True,bY=True,bZ=True))
    Parents=((Item=(Type=Bone,Name="dyn_hair_back_m_parent"),Weight=1),(Item=(Type=Bone,Name="spine_05"),Weight=0))
    Weight=1
   )
   > &RigUnit_ParentConstraint(
    AdvancedSettings=(RotationOrderForFilter=XZY)
    Child=(Type=Bone,Name="dyn_hair_back_l_01")
    Filter=(TranslationFilter=(bX=True,bY=True,bZ=True),RotationFilter=(bX=true,bY=true,bZ=true),ScaleFilter=(bX=True,bY=True,bZ=True))
    Parents=((Item=(Type=Bone,Name="dyn_hair_back_m_02"),Weight=0.5),(Item=(Type=Bone,Name="spine_05"),Weight=0),(Item=(Type=Bone,Name="clavicle_l"),Weight=0.5))
    Weight=1
   )
   > &RigUnit_ParentConstraint(
    AdvancedSettings=(RotationOrderForFilter=XZY)
    Child=(Type=Bone,Name="dyn_hair_back_l_03")
    Parents=((Item=(Type=Bone,Name="spine_04"),Weight=0.002),(Item=(Type=Bone,Name="upperarm_l"),Weight=0.083),(Item=(Type=Bone,Name="spine_05"),Weight=0.1205),(Item=(Type=Bone,Name="clavicle_l"),Weight=0.7945))
    Weight=1
   )
   > &RigUnit_ParentConstraint(
    AdvancedSettings=(RotationOrderForFilter=XZY)
    Child=(Type=Bone,Name="dyn_hair_back_l_04")
    Parents="((Item=(Type=Bone,Name=\"upperarm_l\"),Weight=0.0002),(Item=(Type=Bone,Name=\"spine_03\"),Weight=0.0005),(Item=(Type=Bone,Name=\"spine_02\"),Weight=0.0015),(Item=(Type=Bone,Name=\"clavicle_r\"),Weight=0.0026),(Item=(Type=Bone,Name=\"spine_04\"),Weight=0.1052),(Item= ...(truncated)"
    Weight=1
   )
   > &RigUnit_ParentConstraint(
    AdvancedSettings=(RotationOrderForFilter=XZY)
    Child=(Type=Bone,Name="dyn_hair_back_r_base")
    Filter=(TranslationFilter=(bX=true,bY=true,bZ=true),RotationFilter=(bX=true,bY=true,bZ=true),ScaleFilter=(bX=True,bY=True,bZ=True))
    Parents=((Item=(Type=Bone,Name="dyn_hair_back_m_base"),Weight=1),(Item=(Type=Bone,Name="spine_05"),Weight=0))
    Weight=1
   )
   > &RigUnit_ParentConstraint(
    AdvancedSettings=(RotationOrderForFilter=XZY)
    Child=(Type=Bone,Name="dyn_hair_back_r_parent")
    Filter=(TranslationFilter=(bX=True,bY=True,bZ=True),RotationFilter=(bX=true,bY=true,bZ=true),ScaleFilter=(bX=True,bY=True,bZ=True))
    Parents=((Item=(Type=Bone,Name="dyn_hair_back_m_parent"),Weight=1),(Item=(Type=Bone,Name="spine_05"),Weight=0))
    Weight=1
   )
   > &RigUnit_ParentConstraint(
    AdvancedSettings=(RotationOrderForFilter=XZY)
    Child=(Type=Bone,Name="dyn_hair_back_r_01")
    Filter=(TranslationFilter=(bX=True,bY=True,bZ=True),RotationFilter=(bX=true,bY=true,bZ=true),ScaleFilter=(bX=True,bY=True,bZ=True))
    Parents=((Item=(Type=Bone,Name="dyn_hair_back_r_02"),Weight=0.5),(Item=(Type=Bone,Name="spine_05"),Weight=0),(Item=(Type=Bone,Name="clavicle_r"),Weight=0.5))
    Weight=1
   )
   > &RigUnit_ParentConstraint(
    AdvancedSettings=(RotationOrderForFilter=XZY)
    Child=(Type=Bone,Name="dyn_hair_back_r_03")
    Parents=((Item=(Type=Bone,Name="spine_04"),Weight=0.002),(Item=(Type=Bone,Name="upperarm_r"),Weight=0.083),(Item=(Type=Bone,Name="spine_05"),Weight=0.1205),(Item=(Type=Bone,Name="clavicle_r"),Weight=0.7945))
    Weight=1
   )
   > &RigUnit_ParentConstraint(
    AdvancedSettings=(RotationOrderForFilter=XZY)
    Child=(Type=Bone,Name="dyn_hair_back_r_04")
    Parents="((Item=(Type=Bone,Name=\"upperarm_r\"),Weight=0.0002),(Item=(Type=Bone,Name=\"spine_03\"),Weight=0.0005),(Item=(Type=Bone,Name=\"spine_02\"),Weight=0.0015),(Item=(Type=Bone,Name=\"clavicle_l\"),Weight=0.0026),(Item=(Type=Bone,Name=\"spine_04\"),Weight=0.1052),(Item= ...(truncated)"
    Weight=1
   )
   SequenceAggregate_1 = &RigVMFunction_Sequence(ExecuteContext=@SequenceAggregate.B)
   RigVMFunction_Sequence_1 = RigVMAggregateNode(ExecuteContext=@SequenceAggregate_1.B)
   n1 = &RigUnit_SphericalPoseReader(
    ActiveRegionScaleFactors=(PositiveWidth=1.25,NegativeWidth=1.5,PositiveHeight=1.25,NegativeHeight=2)
    ActiveRegionSize=0.1
    Debug=(bDrawDebug=false,bDraw2D=false,bDrawLocalAxes=False,DebugScale=25,DebugSegments=20,DebugThickness=0.25)
    DriverAxis=(X=1,Y=0,Z=0)
    DriverItem=(Type=Bone,Name="upperarm_l")
    FalloffRegionScaleFactors=(PositiveWidth=1,NegativeWidth=0.5,PositiveHeight=1,NegativeHeight=1)
    FalloffSize=0.25
    FlipHeightScaling=False
    FlipWidthScaling=False
    OptionalParentItem=(Type=Bone,Name="None")
    RotationOffset=(X=70,Y=-20,Z=-35)
    ExecutePin=@RigVMFunction_Sequence_1.A
   )
   > &RigUnit_ModifyTransforms(
    ItemToModify=((Item=(Type=Bone,Name="dyn_hair_back_l_01"),Transform=(Rotation=(X=0,Y=0,Z=0,W=1),Translation=(X=-2.75,Y=0,Z=0.5),Scale3D=(X=1,Y=1,Z=1))))
    WeightMinimum=0
    Weight=@n1.OutputParam
   )
   > &RigUnit_ModifyTransforms(
    ItemToModify=((Item=(Type=Bone,Name="dyn_hair_back_r_01"),Transform=(Rotation=(X=0,Y=0,Z=0,W=1),Translation=(X=-2.75,Y=0,Z=0.5),Scale3D=(X=1,Y=1,Z=1))))
    WeightMinimum=0
    Weight=@n1.OutputParam
   )
   n2 = &RigUnit_SphericalPoseReader(
    ActiveRegionScaleFactors=(PositiveWidth=1,NegativeWidth=1,PositiveHeight=1,NegativeHeight=1.5)
    ActiveRegionSize=0.1
    Debug=(bDrawDebug=false,bDraw2D=false,bDrawLocalAxes=False,DebugScale=25,DebugSegments=20,DebugThickness=0.25)
    DriverAxis=(X=1,Y=0,Z=0)
    DriverItem=(Type=Bone,Name="upperarm_l")
    FalloffRegionScaleFactors=(PositiveWidth=0.5,NegativeWidth=0.5,PositiveHeight=0.5,NegativeHeight=1)
    FalloffSize=0.25
    FlipHeightScaling=False
    FlipWidthScaling=False
    OptionalParentItem=(Type=Bone,Name="None")
    RotationOffset=(X=50,Y=0,Z=-20)
    ExecutePin=@RigVMFunction_Sequence_1.B
   )
   n3 = RigVMVariableNode() {variable=Crounchfix}
   > &RigUnit_ModifyTransforms(
    ItemToModify="((Item=(Type=Bone,Name=\"dyn_hair_back_l_01\"),Transform=(Rotation=(X=0,Y=0,Z=0,W=1),Translation=(X=-2,Y=0,Z=-2),Scale3D=(X=1,Y=1,Z=1))),(Item=(Type=Bone,Name=\"dyn_hair_back_r_01\"),Transform=(Rotation=(X=0,Y=0,Z=0,W=1),Translation=(X=-2,Y=0,Z=-2),Scale3D=(X= ...(truncated)"
    WeightMinimum=0
    Weight=&RigVMFunction_MathDoubleRemap(
     SourceMaximum=1
     SourceMinimum=0
     TargetMaximum=1
     TargetMinimum=0
     bClamp=False
     Value=&RigVMFunction_MathDoubleMul(A=@n2.OutputParam, B=@n3.Value).Result
    ).Result
   )
   RigVMFunction_Sequence = &RigVMFunction_Sequence(ExecuteContext=@SequenceAggregate_1.A)
   n4 = &RigUnit_SphericalPoseReader(
    ActiveRegionScaleFactors=(PositiveWidth=1,NegativeWidth=1,PositiveHeight=1.3,NegativeHeight=1)
    ActiveRegionSize=0.1
    Debug=(bDrawDebug=false,bDraw2D=false,bDrawLocalAxes=False,DebugScale=25,DebugSegments=20,DebugThickness=0.25)
    DriverAxis=(X=1,Y=0,Z=0)
    DriverItem=(Type=Bone,Name="upperarm_l")
    FalloffRegionScaleFactors=(PositiveWidth=1,NegativeWidth=1,PositiveHeight=1,NegativeHeight=1)
    FalloffSize=0.2
    FlipHeightScaling=False
    FlipWidthScaling=False
    OptionalParentItem=(Type=Bone,Name="None")
    RotationOffset=(X=-20,Y=-90,Z=0)
    ExecutePin=@RigVMFunction_Sequence.A
   )
   Multiply_1 = RigVMAggregateNode(C=@n4.OutputParam, A=RigVMVariableNode() {variable=BHairDyn}.Value, B=RigVMVariableNode() {variable=JumpFix}.Value)
   Remap = &RigVMFunction_MathDoubleRemap(SourceMaximum=1, SourceMinimum=0, TargetMaximum=1, TargetMinimum=0, bClamp=False, Value=@Multiply_1.Result)
   > &RigUnit_ModifyTransforms(
    ItemToModify="((Item=(Type=Bone,Name=\"dyn_hair_back_l_03\"),Transform=(Rotation=(X=0,Y=0,Z=0,W=1),Translation=(X=0,Y=-1,Z=1.5),Scale3D=(X=1,Y=1,Z=1))),(Item=(Type=Bone,Name=\"dyn_hair_back_l_02\"),Transform=(Rotation=(X=0,Y=0,Z=-0.0872,W=0.9962),Translation=(X=-1.5,Y=1,Z=0 ...(truncated)"
    WeightMinimum=0
    Weight=@Remap.Result
   )
   n5 = &RigUnit_SphericalPoseReader(
    ActiveRegionScaleFactors=(PositiveWidth=1,NegativeWidth=1.3,PositiveHeight=1,NegativeHeight=1)
    ActiveRegionSize=0.1
    Debug=(bDrawDebug=false,bDraw2D=false,bDrawLocalAxes=False,DebugScale=25,DebugSegments=20,DebugThickness=0.25)
    DriverAxis=(X=-1,Y=0,Z=0)
    DriverItem=(Type=Bone,Name="upperarm_r")
    FalloffRegionScaleFactors=(PositiveWidth=1,NegativeWidth=1,PositiveHeight=1,NegativeHeight=1)
    FalloffSize=0.2
    FlipHeightScaling=False
    FlipWidthScaling=False
    OptionalParentItem=(Type=Bone,Name="None")
    RotationOffset=(X=35,Y=80,Z=10)
    ExecutePin=@RigVMFunction_Sequence.B
   )
   Multiply = RigVMAggregateNode(C=@n5.OutputParam, A=RigVMVariableNode() {variable=BHairDyn}.Value, B=RigVMVariableNode() {variable=JumpFix}.Value)
   > &RigUnit_ModifyTransforms(
    ItemToModify="((Item=(Type=Bone,Name=\"dyn_hair_back_r_03\"),Transform=(Rotation=(X=0,Y=0,Z=0,W=1),Translation=(X=0,Y=1,Z=1.5),Scale3D=(X=1,Y=1,Z=1))),(Item=(Type=Bone,Name=\"dyn_hair_back_r_02\"),Transform=(Rotation=(X=0.013,Y=0.0011,Z=0.0871,W=0.9961),Translation=(X=-1.5, ...(truncated)"
    WeightMinimum=0
    Weight=&RigVMFunction_MathDoubleRemap(SourceMaximum=1, SourceMinimum=0, TargetMaximum=1, TargetMinimum=0, bClamp=False, Value=@Multiply.Result).Result
   )
   &RigUnit_ModifyTransforms(
    ExecutePin=()
    ItemToModify=((Item=(Type=Null,Name="dyn_hair_back_m_parent_null"),Transform=(Rotation=(X=0,Y=0,Z=0,W=1),Translation=(X=-20,Y=0,Z=0),Scale3D=(X=1,Y=1,Z=1))))
    Weight=1
    WeightMinimum=0
   )
  ]}
 RigVMModel Controls {nodes [
   n1 = &RigUnit_PrepareForExecution()
   SequenceAggregate = RigVMAggregateNode(ExecuteContext=@n1.ExecutePin)
   RerouteNode_6_1 = RigVMDispatchNode(Values=((Type=Control,Name="dyn_hair_back_m_01_ctrl"),((),Type=Control,Name="dyn_hair_back_m_02_ctrl"),((),Type=Control,Name="dyn_hair_back_m_03_ctrl"),((),Type=Control,Name="dyn_hair_back_m_04_ctrl"),((),Type=Control,Name="dyn_hair_back_m_05_ctrl"))) {template="DISPATCH_RigVMDispatch_ArrayMake(in Values,out Array)"}
   RigVMVariableNode(Value=@RerouteNode_6_1.Array, ExecuteContext=@SequenceAggregate.A) {variable=dyn_hair_Middle_ctrls}
   > RigVMVariableNode(Value=/Script/ControlRig.RigUnit_HierarchyGetPoseItemArray(ElementType=All, Initial=false, ItemsToGet=@RerouteNode_6_1.Array) {is=struct RigVMUnitNode}.Pose) {variable=dyn_hair_Middle_pose}
  ]}
]
hierarchy [
 root {loc=<0 0 0>, rot=<0 0 0 1>}
 pelvis {parent=root, loc=<0 -0.5301 98.0571>, rot=<0.0127 0.707 -0.0127 -0.707>}
 spine_01 {parent=pelvis, loc=<6.1208 -0.9328 0>, rot=<0 0 0.0576 -0.9983>}
 spine_02 {parent=spine_01, loc=<6.6022 -0.0153 0>, rot=<0 0 0.0022 -1>}
 spine_03 {parent=spine_02, loc=<6.4919 -0.0105 0>, rot=<0 0 0.0015 -1>}
 spine_04 {parent=spine_03, loc=<6.9596 -0.007 0>, rot=<0 0 0.0009 -1>}
 spine_05 {parent=spine_04, loc=<6.9981 -0.0017 0>, rot=<0 0 -0.0042 -1>}
 neck_01 {parent=spine_05, loc=<11.3138 0.1529 0>, rot=<0 0 -0.2003 -0.9797>}
 neck_02 {parent=neck_01, loc=<4.5627 0 0>, rot=<0 0 0.0504 -0.9987>}
 head {parent=neck_02, loc=<5.3914 0 0>, rot=<0 0 0.115 -0.9934>}
 dyn_hair_bang_l_base {parent=head, loc=<3.5485 -0.758 -9.956>, rot=<0.0029 -0.7071 -0.0029 -0.7071>}
 dyn_hair_bang_l_parent {parent=dyn_hair_bang_l_base, loc=<0 0 0>, rot=<0 0 0 1>}
 dyn_hair_side_r_base {parent=head, loc=<0.6982 2.5017 8.8512>, rot=<0.0109 -0.9768 0.2076 0.0512>}
 dyn_hair_side_r_parent {parent=dyn_hair_side_r_base, loc=<0 0 0>, rot=<0 0 0 1>}
 dyn_hair_side_r_01 {parent=dyn_hair_side_r_parent, loc=<3.5248 -0.0627 0.0788>, rot=<0.0697 0.0174 0.0012 -0.9974>}
 dyn_hair_side_r_02 {parent=dyn_hair_side_r_01, loc=<4.1295 -0.0009 0.31>, rot=<0 0 0 1>}
 dyn_hair_side_r_03 {parent=dyn_hair_side_r_02, loc=<3.6307 -0.0223 0.5031>, rot=<0 0 0 1>}
 dyn_hair_back_r_base {parent=head, loc=<3.1226 -5.6768 6.5996>, rot=<0.0813 -0.87 -0.475 0.1044>}
 dyn_hair_back_r_parent {parent=dyn_hair_back_r_base, loc=<0 0 0>, rot=<0 0 0 1>}
 dyn_hair_back_r_01 {parent=dyn_hair_back_r_parent, loc=<5.3186 -0.5291 1.3299>, rot=<0 0 0 1>}
 dyn_hair_back_r_02 {parent=dyn_hair_back_r_01, loc=<3.9918 -0.9775 -0.3881>, rot=<0 0 0 1>}
 dyn_hair_back_r_03 {parent=dyn_hair_back_r_02, loc=<4.1773 -1.0541 0.477>, rot=<0 0 0 1>}
 dyn_hair_back_r_04 {parent=dyn_hair_back_r_03, loc=<5.6033 -1.1159 -1.8734>, rot=<0 0 0 1>}
 dyn_hair_back_r_06 {parent=dyn_hair_back_r_04, loc=<4.0484 3.2364 1.3059>, rot=<0 0.2588 0 -0.9659>}
 dyn_hair_back_r_05 {parent=dyn_hair_back_r_04, loc=<3.2529 0.67 1.9325>, rot=<-0.0153 0.3029 0.07 -0.9503>}
 dyn_hair_back_r_04_null {parent=dyn_hair_back_r_04, type=Null, loc=<0 0 0>, rot=<0 0 0 1>}
 dyn_hair_back_r_03_null {parent=dyn_hair_back_r_03, type=Null, loc=<0 0 0>, rot=<0 0 0 1>}
 dyn_hair_back_r_02_null {parent=dyn_hair_back_r_02, type=Null, loc=<0 0 0>, rot=<0 0 0 1>}
 dyn_hair_back_r_01_null {parent=dyn_hair_back_r_01, type=Null, loc=<0 0 0>, rot=<0 0 0 1>}
 dyn_hair_back_r_parent_null {parent=dyn_hair_back_r_parent, type=Null, loc=<0 0 0>, rot=<0 0 0 1>}
 dyn_hair_back_r_base_null {parent=dyn_hair_back_r_base, type=Null, loc=<0 0 0>, rot=<0 0 0 1>}
 dyn_hair_leaf_03 {parent=head, loc=<10.5836 1.747 -7.8712>, rot=<-0.1726 0.0542 0.5575 -0.8102>}
 dyn_hair_back_m_base {parent=head, loc=<3.3054 -8.526 -0.3617>, rot=<-0.1471 0.6397 0.749 -0.0905>}
 dyn_hair_back_m_parent {parent=dyn_hair_back_m_base, loc=<0 0 0>, rot=<0 0 0 1>}
 dyn_hair_back_m_01 {parent=dyn_hair_back_m_parent, loc=<0 0 0>, rot=<0 0 0 1>}
 dyn_hair_back_m_02 {parent=dyn_hair_back_m_01, loc=<4.0992 0.3557 0.3929>, rot=<-0.0872 0 0 -0.9962>}
 dyn_hair_back_m_03 {parent=dyn_hair_back_m_02, loc=<4.1531 0.4067 0.5332>, rot=<-0.0436 -0.0019 -0.0436 -0.9981>}
 dyn_hair_back_m_04 {parent=dyn_hair_back_m_03, loc=<5.2163 -0.1838 0.7025>, rot=<0 0 0 1>}
 dyn_hair_back_m_05 {parent=dyn_hair_back_m_04, loc=<3.391 0.091 1.3159>, rot=<0 0 0 1>}
 dyn_hair_back_m_06 {parent=dyn_hair_back_m_05, loc=<4.448 4.8265 -1.2215>, rot=<0.0427 0.2077 0.0091 -0.9772>}
 dyn_hair_back_m_07 {parent=dyn_hair_back_m_05, loc=<6.3451 -1.1559 1.7064>, rot=<0 0 0.0698 -0.9976>}
 dyn_hair_back_m_05_null {parent=dyn_hair_back_m_05, type=Null, loc=<0 0 0>, rot=<0 0 0 1>}
 dyn_hair_back_m_04_null {parent=dyn_hair_back_m_04, type=Null, loc=<0 0 0>, rot=<0 0 0 1>}
 dyn_hair_back_m_03_null {parent=dyn_hair_back_m_03, type=Null, loc=<0 0 0>, rot=<0 0 0 1>}
 dyn_hair_back_m_02_null {parent=dyn_hair_back_m_02, type=Null, loc=<0 0 0>, rot=<0 0 0 1>}
 dyn_hair_back_m_01_null {parent=dyn_hair_back_m_01, type=Null, loc=<0 0 0>, rot=<0 0 0 1>}
 dyn_hair_back_m_parent_null {parent=dyn_hair_back_m_parent, type=Null, loc=<0 0 0>, rot=<0 0 0 1>}
 dyn_hair_back_m_base_null {parent=dyn_hair_back_m_base, type=Null, loc=<0 0 0>, rot=<0 0 0 1>}
 dyn_hair_leaf_04 {parent=head, loc=<5.4668 -1.8989 -9.1273>, rot=<-0.176 -0.0893 0.7352 -0.6485>}
 dyn_hair_back_l_base {parent=head, loc=<4.3519 -4.1845 -6.9958>, rot=<-0.0821 0.2488 0.9549 -0.1398>}
 dyn_hair_back_l_parent {parent=dyn_hair_back_l_base, loc=<0 0 0>, rot=<0 0 0 1>}
 dyn_hair_back_l_01 {parent=dyn_hair_back_l_parent, loc=<4.9044 -0.8975 1.5243>, rot=<0 0 0 1>}
 dyn_hair_back_l_02 {parent=dyn_hair_back_l_01, loc=<3.877 0.8012 -0.4498>, rot=<-0.1337 -0.2594 -0.0914 -0.9521>}
 dyn_hair_back_l_03 {parent=dyn_hair_back_l_02, loc=<3.0791 1.8093 0.9682>, rot=<-0.21 0.2362 -0.0524 -0.9473>}
 dyn_hair_back_l_04 {parent=dyn_hair_back_l_03, loc=<5.2678 0.3899 0.7156>, rot=<-0.0872 0 0 -0.9962>}
 dyn_hair_back_l_06 {parent=dyn_hair_back_l_04, loc=<7.5412 1.1145 2.7785>, rot=<0.1361 0.2929 0.1473 -0.9349>}
 dyn_hair_back_l_05 {parent=dyn_hair_back_l_04, loc=<4.6521 0.486 1.1305>, rot=<-0.0324 -0.2945 0.2485 -0.9222>}
 dyn_hair_back_l_07 {parent=dyn_hair_back_l_04, loc=<2.6796 -2.2864 0.6881>, rot=<0.04 0.0966 0.3806 -0.9188>}
 dyn_hair_back_l_04_null {parent=dyn_hair_back_l_04, type=Null, loc=<0 0 0>, rot=<0 0 0 1>}
 dyn_hair_back_l_03_null {parent=dyn_hair_back_l_03, type=Null, loc=<0 0 0>, rot=<0 0 0 1>}
 dyn_hair_back_l_02_null {parent=dyn_hair_back_l_02, type=Null, loc=<0 0 0>, rot=<0 0 0 1>}
 dyn_hair_back_l_01_null {parent=dyn_hair_back_l_01, type=Null, loc=<0 0 0>, rot=<0 0 0 1>}
 dyn_hair_back_l_parent_null {parent=dyn_hair_back_l_parent, type=Null, loc=<0 0 0>, rot=<0 0 0 1>}
 dyn_hair_back_l_base_null {parent=dyn_hair_back_l_base, type=Null, loc=<0 0 0>, rot=<0 0 0 1>}
 dyn_hair_side_l_base {parent=head, loc=<2.2249 1.9643 -7.6992>, rot=<0.0091 -0.1734 0.9835 -0.0515>}
 dyn_hair_side_l_parent {parent=dyn_hair_side_l_base, loc=<0 0 0>, rot=<0 0 0 1>}
 dyn_hair_side_l_07 {parent=dyn_hair_side_l_parent, loc=<0.6163 -3.5304 -0.6736>, rot=<0.1677 -0.0664 0.2872 -0.9407>}
 dyn_hair_side_l_05 {parent=dyn_hair_side_l_parent, loc=<5.9002 2.6808 -0.7973>, rot=<0.0657 -0.1179 -0.685 -0.7159>}
 dyn_hair_side_l_04 {parent=dyn_hair_side_l_parent, loc=<2.8832 4.8107 -0.6176>, rot=<-0.0965 -0.059 -0.8065 -0.5803>}
 dyn_hair_side_l_01 {parent=dyn_hair_side_l_parent, loc=<5.291 -0.9571 0.8568>, rot=<0 0 0 1>}
 dyn_hair_side_l_02 {parent=dyn_hair_side_l_01, loc=<4.0234 1.0836 0.0666>, rot=<0 0 0 1>}
 dyn_hair_side_l_03 {parent=dyn_hair_side_l_02, loc=<1.2333 2.8378 -2.0966>, rot=<-0.1318 -0.5242 -0.1904 -0.8195>}
 dyn_hair_side_l_06 {parent=dyn_hair_side_l_parent, loc=<5.6315 -1.9971 -0.771>, rot=<0 0 0 1>}
 dyn_hair_front_r_base {parent=head, loc=<0.5346 7.4393 6.821>, rot=<-0.0171 -0.978 0.2079 0.0036>}
 dyn_hair_front_r_parent {parent=dyn_hair_front_r_base, loc=<0 0 0>, rot=<0 0 0 1>}
 dyn_hair_front_r_01 {parent=dyn_hair_front_r_parent, loc=<4.6702 -0.4052 1.2033>, rot=<0 0.1392 0 -0.9903>}
 dyn_hair_front_r_02 {parent=dyn_hair_front_r_01, loc=<4.5846 -0.0567 -0.3233>, rot=<-0.1736 0 0 -0.9848>}
 dyn_hair_front_r_03 {parent=dyn_hair_front_r_02, loc=<3.6874 -1.317 0.3788>, rot=<-0.0818 -0.3285 -0.0152 -0.9408>}
 dyn_hair_front_r_04 {parent=dyn_hair_front_r_03, loc=<5.2689 2.2587 -0.6207>, rot=<-0.2585 0.3078 -0.4266 -0.8102>}
 dyn_hair_front_r_05 {parent=dyn_hair_front_r_01, loc=<3.6188 -1.6992 -3.8275>, rot=<-0.2611 -0.3946 -0.0479 -0.8797>}
 dyn_hair_leaf_02 {parent=head, loc=<4.2314 2.828 9.1707>, rot=<0.8715 -0.4441 0.1853 0.0944>}
 dyn_hair_leaf_01 {parent=head, loc=<10.7547 5.0017 7.7473>, rot=<0.8452 -0.4192 -0.0775 0.3222>}
 dyn_hair_bang_base {parent=head, loc=<11.6568 9.849 -0.9329>, rot=<0.2994 0.3185 -0.3978 -0.8067>}
 dyn_hair_bang_parent {parent=dyn_hair_bang_base, loc=<0 0 0>, rot=<0 0 0 1>}
 dyn_hair_bang_01 {parent=dyn_hair_bang_parent, loc=<1.7033 -0.1008 0.3388>, rot=<0 0.682 0 -0.7314>}
 dyn_hair_bang_02 {parent=dyn_hair_bang_01, loc=<4.8843 0.9144 2.1918>, rot=<0.5263 0.1199 0.8295 0.1436>}
 dyn_hair_bang_03 {parent=dyn_hair_bang_02, loc=<4.817 0.7072 1.3075>, rot=<0 0.4226 0 -0.9063>}
 faceAttach {parent=head, loc=<-3.5774 -0.1155 -0.025>, rot=<0.0082 -0.7071 -0.0082 -0.7071>, scale=<0.89 0.89 0.89>}
 C_brow_mid {parent=faceAttach, loc=<-0.0568 9.7012 8.2927>, rot=<0 0.0116 -0.0023 -0.9999>}
 L_brow_mid {parent=faceAttach, loc=<3.3744 9.0398 9.7829>, rot=<-0.0001 -0.0041 0.2553 -0.9669>}
 L_brow_outer {parent=faceAttach, loc=<5.8167 6.9935 9.6225>, rot=<0.1507 0.1859 0.3939 -0.8875>}
 L_cheek_inner {parent=faceAttach, loc=<4.9296 7.3906 2.7776>, rot=<-0.0302 -0.1674 0.3306 -0.9283>}
 C_lip_upper_mid {parent=faceAttach, loc=<-0.0316 9.4307 0.9534>, rot=<0.2225 0.0021 -0.0008 -0.9749>}
 L_lip_upper_outer {parent=faceAttach, loc=<1.7144 8.3268 1.0401>, rot=<0.0359 -0.0729 0.2628 -0.9614>}
 L_lip_corner {parent=faceAttach, loc=<2.4852 8.283 0.9557>, rot=<0.0025 0.0085 -0.0146 -0.9999>}
 C_nose_bridge {parent=faceAttach, loc=<-0.0484 10.8315 5.5806>, rot=<-0.2217 -0.0273 -0.0077 -0.9747>}
 R_brow_mid {parent=faceAttach, loc=<-3.3933 9.0563 9.7842>, rot=<-0.0001 0.0042 -0.2553 -0.9669>}
 R_brow_outer {parent=faceAttach, loc=<-5.8684 7.0182 9.5816>, rot=<0.0839 -0.1017 -0.419 -0.8984>}
 R_cheek_inner {parent=faceAttach, loc=<-4.9746 7.3999 2.7672>, rot=<-0.0306 0.1684 -0.3305 -0.9282>}
 R_lip_corner {parent=faceAttach, loc=<-2.5253 8.1671 0.9376>, rot=<0.0028 -0.0081 0.0146 -0.9999>}
 C_jaw {parent=faceAttach, loc=<-0.0877 3.2443 0.4936>, rot=<0 0 0 1>}
 L_lip_lower_outer {parent=C_jaw, loc=<1.8342 5.15 0.4922>, rot=<0.0163 0.0628 0.3194 -0.9454>}
 C_lip_lower_mid {parent=C_jaw, loc=<0.0518 6.1032 0.4282>, rot=<0 -0.0217 -0.0046 -0.9998>}
 R_lip_lower_outer {parent=C_jaw, loc=<-1.6846 5.4414 0.4392>, rot=<0.0071 -0.0129 -0.2815 -0.9595>}
 R_eye {parent=faceAttach, loc=<-3.2819 6.6673 8.1058>, rot=<0 0 0.9998 -0.0187>}
 R_eye_lid_upper_mid {parent=R_eye, loc=<-0.4276 -2.1496 0.6456>, rot=<0.0523 -0.0009 0.0167 -0.9985>}
 R_eye_lid_lower_mid {parent=R_eye, loc=<-0.3833 -1.8172 -0.7242>, rot=<0.0329 -0.0005 0.0167 -0.9993>}
 L_eye {parent=faceAttach, loc=<3.2256 6.6673 8.1058>, rot=<0 0 0.9998 0.0187>}
 L_eye_lid_upper_mid {parent=L_eye, loc=<0.2652 -2.2017 0.6378>, rot=<0.0523 0 0 -0.9986>}
 L_eye_lid_lower_mid {parent=L_eye, loc=<0.3971 -1.7971 -0.7695>, rot=<0.0329 0 0 -0.9995>}
 R_lip_upper_outer {parent=faceAttach, loc=<-1.7802 8.373 1.0842>, rot=<0.0396 0.0624 -0.2588 -0.9631>}
 clavicle_l {parent=spine_05, loc=<5.4831 -0.1053 -3.2946>, rot=<0.0938 -0.7388 0.0307 -0.6666>}
 upperarm_l {parent=clavicle_l, loc=<9.9486 0 0>, rot=<0.0432 -0.2973 -0.0725 -0.951>}
 upperarm_twist_01_l {parent=upperarm_l, loc=<7.2131 0.1149 -0.0692>, rot=<-0.0062 -0.0048 -0.0158 -0.9998>}
 upperarm_twist_02_l {parent=upperarm_l, loc=<19.7925 0.513 -0.1874>, rot=<-0.0062 -0.0048 -0.0158 -0.9998>}
 clavicle_r {parent=spine_05, loc=<5.4831 -0.1053 3.2946>, rot=<0.7388 0.0938 -0.6666 -0.0307>}
 upperarm_r {parent=clavicle_r, loc=<-9.9486 0 0>, rot=<0.0432 -0.2973 -0.0725 -0.951>}
 upperarm_twist_01_r {parent=upperarm_r, loc=<-7.2131 -0.1149 0.0692>, rot=<-0.0062 -0.0048 -0.0158 -0.9998>}
 upperarm_twist_02_r {parent=upperarm_r, loc=<-19.7925 -0.513 0.1874>, rot=<-0.0062 -0.0048 -0.0158 -0.9998>}
 dyn_hair_back_r_base_ctrl {controlType=EULER_TRANSFORM, type=Control, loc=<0 0 0>, rot=<0 0 0 1>}
 dyn_hair_back_r_parent_ctrl {controlType=EULER_TRANSFORM, type=Control, loc=<-6.5996 -7.3898 155.1098>, rot=<-0.3911 0.6879 0.2812 0.543>}
 dyn_hair_back_r_01_ctrl {controlType=EULER_TRANSFORM, type=Control, loc=<-11.327 -7.3022 148.3218>, rot=<-0.3911 0.6879 0.2812 0.543>}
 dyn_hair_back_r_02_ctrl {controlType=EULER_TRANSFORM, type=Control, loc=<-9.7976 -8.6212 146.3148>, rot=<-0.3911 0.6879 0.2812 0.543>}
 dyn_hair_back_r_03_ctrl {controlType=EULER_TRANSFORM, type=Control, loc=<-9.2793 -9.6941 141.33>, rot=<-0.3911 0.6879 0.2812 0.543>}
 dyn_hair_back_r_04_ctrl {controlType=EULER_TRANSFORM, type=Control, loc=<-11.3372 -13.0541 138.7769>, rot=<-0.3911 0.6879 0.2812 0.543>}
 dyn_hair_back_m_base_ctrl {controlType=EULER_TRANSFORM, type=Control, loc=<0 0 0>, rot=<0 0 0 1>}
 dyn_hair_back_m_01_ctrl {controlType=EULER_TRANSFORM, type=Control, loc=<0.3617 -10.2403 155.2691>, rot=<-0.632 0.5146 0.4277 0.3909>}
 dyn_hair_back_m_02_ctrl {controlType=EULER_TRANSFORM, type=Control, loc=<-3.7136 14.7901 150.5819>, rot=<-0.632 0.581 0.3319 0.3909>}
 dyn_hair_back_m_03_ctrl {controlType=EULER_TRANSFORM, type=Control, loc=<-4.7893 0.7534 147.0875>, rot=<-0.5454 0.6306 0.3276 0.4445>}
 dyn_hair_back_m_04_ctrl {controlType=EULER_TRANSFORM, type=Control, loc=<-4.4232 -12.85 141.2659>, rot=<-0.5518 0.5922 0.3748 0.452>}
 dyn_hair_back_m_parent_ctrl {controlType=EULER_TRANSFORM, parent=dyn_hair_back_m_04_ctrl, type=Control, loc=<0.3617 -10.2403 155.2691>, rot=<-0.632 0.5146 0.4277 0.3909>}
 dyn_hair_back_m_05_ctrl {controlType=EULER_TRANSFORM, type=Control, loc=<-2.4687 -12.9418 137.0209>, rot=<-0.5518 0.5922 0.3748 0.452>}
 dyn_hair_back_l_base_ctrl {controlType=EULER_TRANSFORM, type=Control, loc=<0 0 0>, rot=<0 0 0 1>}
 dyn_hair_back_l_parent_ctrl {controlType=EULER_TRANSFORM, type=Control, loc=<6.9958 -5.9077 156.3514>, rot=<0.7329 -0.2722 -0.6183 -0.0801>}
 dyn_hair_back_l_01_ctrl {controlType=EULER_TRANSFORM, type=Control, loc=<1.6509 -5.0364 149.6322>, rot=<0.7329 -0.2722 -0.6183 -0.0801>}
 dyn_hair_back_l_02_ctrl {controlType=EULER_TRANSFORM, type=Control, loc=<-69.1602 36.1251 126.3362>, rot=<-0.8461 0.5214 0.0767 0.0801>}
 dyn_hair_back_l_03_ctrl {controlType=EULER_TRANSFORM, type=Control, loc=<70.5084 41.9075 119.6914>, rot=<0.5528 -0.6905 -0.4641 -0.0471>}
 dyn_hair_back_l_04_ctrl {controlType=EULER_TRANSFORM, type=Control, loc=<0.5037 11.6319 139.8144>, rot=<0.6596 -0.588 -0.3599 -0.2995>}
]
variables [
 BHairDyn {type=double, default=0}
 Crounchfix {type=double, default=0}
 JumpFix {type=double, default=0}
 dyn_hair_Left_ctrls {
  type="TArray<FRigElementKey>"
  default=((Type=Control,Name="dyn_hair_back_l_02"),(Type=Control,Name="dyn_hair_back_l_03"),(Type=Control,Name="dyn_hair_back_l_04"))
 }
 dyn_hair_Left_pose {type=FRigElementKey, default=(Type=None,Name="")}
 dyn_hair_Middle_ctrls {
  type="TArray<FRigElementKey>"
  default=((Type=Control,Name="dyn_hair_back_m_02"),(Type=Control,Name="dyn_hair_back_m_03"),(Type=Control,Name="dyn_hair_back_m_04"),(Type=Control,Name="dyn_hair_back_m_05"))
 }
 dyn_hair_Middle_pose {type=FRigPose, default=(Elements=,HierarchyTopologyVersion=-1,PoseHash=-1)}
 dyn_hair_Right_ctrls {
  type="TArray<FRigElementKey>"
  default=((Type=Control,Name="dyn_hair_back_r_02"),(Type=Control,Name="dyn_hair_back_r_03"),(Type=Control,Name="dyn_hair_back_r_04"))
 }
 dyn_hair_Right_pose {type="TArray<FRigPose>", default=(())}
]
