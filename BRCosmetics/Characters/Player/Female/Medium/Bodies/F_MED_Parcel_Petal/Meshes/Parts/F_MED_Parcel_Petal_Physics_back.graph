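# Physics asset for the back hair of the F_MED_Parcel_Petal female character cosmetic, with capsule bodies for hair and upper body.
# Left, middle and right dyn_hair_back chains hang from the head, each forking at the tip, with narrow twist and wider swing limits at the forks.
# Head, upper spine, clavicle and upper-arm capsules form the body volume, linked by free angular joints except the tightly limited head-to-spine joint.
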PhysicsAsset /BRCosmetics/Characters/Player/Female/Medium/Bodies/F_MED_Parcel_Petal/Meshes/Parts/F_MED_Parcel_Petal_Physics_back
bodies {
 clavicle_l [{shapeType=Capsule, length=15, radius=5.5, shapeName=clavicle_l_capsule, center=<5.6258 0 0.0825>, rotation=<rot 90 0 0>}]
 clavicle_r [{shapeType=Capsule, length=15, radius=5.5, shapeName=clavicle_r_capsule, center=<-6.8311 0 0.0289>, rotation=<rot 90 0 0>}]
 dyn_hair_back_l_01 [{shapeType=Capsule, length=2, radius=2, shapeName=dyn_hair_back_l_01_capsule, center=<2 0 0>, rotation=<rot 90 0 0>}]
 dyn_hair_back_l_02 [{shapeType=Capsule, length=2, radius=3.5, shapeName=dyn_hair_back_l_02_capsule, center=<2 0 0>, rotation=<rot 90 0 0>}]
 dyn_hair_back_l_03 [{shapeType=Capsule, length=2, radius=2, shapeName=dyn_hair_back_l_03_capsule, center=<2 0 0>, rotation=<rot 90 0 0>}]
 dyn_hair_back_l_04 [{shapeType=Capsule, length=2, radius=2.5, shapeName=dyn_hair_back_l_04_capsule, center=<2 0 0>, rotation=<rot 90 0 0>}]
 dyn_hair_back_l_05 [{shapeType=Capsule, length=2, radius=2.5, shapeName=dyn_hair_back_l_05_capsule, center=<2 0 0>, rotation=<rot 90 0 0>}]
 dyn_hair_back_l_06 [{shapeType=Capsule, length=2, radius=2.5, shapeName=dyn_hair_back_l_06_capsule, center=<2 0 0>, rotation=<rot 90 0 0>}]
 dyn_hair_back_l_07 [{shapeType=Capsule, length=2, radius=2.5, shapeName=dyn_hair_back_l_07_capsule, center=<2 0 0>, rotation=<rot 90 0 0>}]
 dyn_hair_back_m_01 [{shapeType=Capsule, length=2, radius=2, shapeName=dyn_hair_back_m_01_capsule, center=<2 0 0>, rotation=<rot 90 0 0>}]
 dyn_hair_back_m_03 [{shapeType=Capsule, length=2, radius=2, shapeName=dyn_hair_back_m_03_capsule, center=<2 0 0>, rotation=<rot 90 0 0>}]
 dyn_hair_back_m_04 [{shapeType=Capsule, length=2, radius=2, shapeName=dyn_hair_back_m_04_capsule, center=<2 0 0>, rotation=<rot 90 0 0>}]
 dyn_hair_back_m_05 [{shapeType=Capsule, length=2, radius=2.5, shapeName=dyn_hair_back_m_05_capsule, center=<2 0 0>, rotation=<rot 90 0 0>}]
 dyn_hair_back_m_06 [{shapeType=Capsule, length=2, radius=2.5, shapeName=dyn_hair_back_m_06_capsule, center=<2 0 0>, rotation=<rot 90 0 0>}]
 dyn_hair_back_m_07 [{shapeType=Capsule, length=2, radius=2.5, shapeName=dyn_hair_back_m_07_capsule, center=<2 0 0>, rotation=<rot 90 0 0>}]
 dyn_hair_back_r_01 [{shapeType=Capsule, length=2, radius=2, shapeName=dyn_hair_back_r_01_capsule, center=<2 0 0>, rotation=<rot 90 0 0>}]
 dyn_hair_back_r_02 [{shapeType=Capsule, length=2, radius=2.5, shapeName=dyn_hair_back_r_02_capsule, center=<2 0 0>, rotation=<rot 90 0 0>}]
 dyn_hair_back_r_03 [{shapeType=Capsule, length=2, radius=2, shapeName=dyn_hair_back_r_03_capsule, center=<2 0 0>, rotation=<rot 90 0 0>}]
 dyn_hair_back_r_04 [{shapeType=Capsule, length=2, radius=2.5, shapeName=dyn_hair_back_r_04_capsule, center=<2 0 0>, rotation=<rot 90 0 0>}]
 dyn_hair_back_r_05 [{shapeType=Capsule, length=2, radius=2.5, shapeName=dyn_hair_back_r_05_capsule, center=<2 0 0>, rotation=<rot 90 0 0>}]
 dyn_hair_back_r_06 [{shapeType=Capsule, length=2, radius=2.5, shapeName=dyn_hair_back_r_06_capsule, center=<2 0 0>, rotation=<rot 90 0 0>}]
 head [
  {shapeType=Capsule, length=5, radius=8.5, shapeName=head_capsule, center=<0 1.8113 0>, rotation=<rot 90 0 0>}
  {shapeType=Capsule, length=51, radius=7, shapeName=head_capsule, center=<-2.7783 0 0>}
 ]
 spine_04 [
  {shapeType=Capsule, length=6, radius=10.5, shapeName=spine_04_capsule, center=<0.2626 2.6545 0>, rotation=<rot 0 -90 0>}
  {shapeType=Capsule, length=20, radius=10, shapeName=spine_04_capsule, center=<0 -2.0108 0>}
 ]
 spine_05 [{shapeType=Capsule, length=74.5, radius=8, shapeName=spine_05_capsule, center=<3.0615 0.4837 0>, rotation=<rot 0 -90 0>}]
 upperarm_l [{shapeType=Capsule, length=20, radius=5, shapeName=upperarm_l_capsule, center=<10 0 0>, rotation=<rot 90 0 0>}]
 upperarm_r [{shapeType=Capsule, length=20, radius=5, shapeName=upperarm_r_capsule, center=<-10 0 0>, rotation=<rot 90 0 0>}]
}
constraints [
 {
  bone1Name=spine_05
  bone2Name=spine_04
  swing1LimitDegrees=45
  swing1Motion=Free
  swing2LimitDegrees=45
  swing2Motion=Free
  twistLimitDegrees=45
  twistMotion=Free
 }
 {
  bone1Name=clavicle_l
  bone2Name=spine_05
  swing1LimitDegrees=45
  swing1Motion=Free
  swing2LimitDegrees=45
  swing2Motion=Free
  twistLimitDegrees=45
  twistMotion=Free
 }
 {
  bone1Name=clavicle_r
  bone2Name=spine_05
  swing1LimitDegrees=45
  swing1Motion=Free
  swing2LimitDegrees=45
  swing2Motion=Free
  twistLimitDegrees=45
  twistMotion=Free
 }
 {
  bone1Name=upperarm_l
  bone2Name=clavicle_l
  swing1LimitDegrees=45
  swing1Motion=Free
  swing2LimitDegrees=45
  swing2Motion=Free
  twistLimitDegrees=45
  twistMotion=Free
 }
 {
  bone1Name=upperarm_r
  bone2Name=clavicle_r
  swing1LimitDegrees=45
  swing1Motion=Free
  swing2LimitDegrees=45
  swing2Motion=Free
  twistLimitDegrees=45
  twistMotion=Free
 }
 {bone1Name=head, bone2Name=spine_05, swing1LimitDegrees=1, swing2LimitDegrees=1, twistLimitDegrees=1}
 {bone1Name=dyn_hair_back_l_07, bone2Name=dyn_hair_back_l_04, swing1LimitDegrees=20, swing2LimitDegrees=20, twistLimitDegrees=5}
 {bone1Name=dyn_hair_back_l_06, bone2Name=dyn_hair_back_l_04, swing1LimitDegrees=20, swing2LimitDegrees=20, twistLimitDegrees=5}
 {bone1Name=dyn_hair_back_l_05, bone2Name=dyn_hair_back_l_04, swing1LimitDegrees=10, swing2LimitDegrees=10, twistLimitDegrees=5}
 {bone1Name=dyn_hair_back_l_04, bone2Name=dyn_hair_back_l_03, swing1LimitDegrees=10, swing2LimitDegrees=10, twistLimitDegrees=5}
 {bone1Name=dyn_hair_back_l_03, bone2Name=dyn_hair_back_l_02, swing1LimitDegrees=10, swing2LimitDegrees=10, twistLimitDegrees=5}
 {bone1Name=dyn_hair_back_l_02, bone2Name=dyn_hair_back_l_01, swing1LimitDegrees=10, swing2LimitDegrees=10, twistLimitDegrees=5}
 {bone1Name=dyn_hair_back_m_07, bone2Name=dyn_hair_back_m_05, swing1LimitDegrees=20, swing2LimitDegrees=20, twistLimitDegrees=5}
 {bone1Name=dyn_hair_back_m_06, bone2Name=dyn_hair_back_m_05, swing1LimitDegrees=20, swing2LimitDegrees=20, twistLimitDegrees=5}
 {bone1Name=dyn_hair_back_m_05, bone2Name=dyn_hair_back_m_04, swing1LimitDegrees=10, swing2LimitDegrees=10, twistLimitDegrees=5}
 {bone1Name=dyn_hair_back_m_04, bone2Name=dyn_hair_back_m_03, swing1LimitDegrees=10, swing2LimitDegrees=10, twistLimitDegrees=5}
 {bone1Name=dyn_hair_back_m_03, bone2Name=dyn_hair_back_m_01, swing1LimitDegrees=10, swing2LimitDegrees=10, twistLimitDegrees=5}
 {bone1Name=dyn_hair_back_r_06, bone2Name=dyn_hair_back_r_04, swing1LimitDegrees=20, swing2LimitDegrees=20, twistLimitDegrees=5}
 {bone1Name=dyn_hair_back_r_05, bone2Name=dyn_hair_back_r_04, swing1LimitDegrees=20, swing2LimitDegrees=20, twistLimitDegrees=5}
 {bone1Name=dyn_hair_back_r_04, bone2Name=dyn_hair_back_r_03, swing1LimitDegrees=10, swing2LimitDegrees=10, twistLimitDegrees=5}
 {bone1Name=dyn_hair_back_r_03, bone2Name=dyn_hair_back_r_02, swing1LimitDegrees=10, swing2LimitDegrees=10, twistLimitDegrees=5}
 {bone1Name=dyn_hair_back_r_02, bone2Name=dyn_hair_back_r_01, swing1LimitDegrees=10, swing2LimitDegrees=10, twistLimitDegrees=5}
 {bone1Name=dyn_hair_back_l_01, bone2Name=head, swing1LimitDegrees=10, swing2LimitDegrees=10, twistLimitDegrees=5}
 {bone1Name=dyn_hair_back_m_01, bone2Name=head, swing1LimitDegrees=10, swing2LimitDegrees=10, twistLimitDegrees=5}
 {bone1Name=dyn_hair_back_r_01, bone2Name=head, swing1LimitDegrees=10, swing2LimitDegrees=10, twistLimitDegrees=5}
]
