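# Post-process style Anim Blueprint for the Parcel Petal female body part, adding secondary motion to hair and hair leaves.
# It copies the parent mesh pose through a Control Rig and front RigidBody, then blends back-hair AnimDynamics (skydive floating) or RigidBody (otherwise).
# Leaf AnimDynamics run last.
# The EventGraph computes movement-state alphas, but the idle and ADS bone offsets they drive sit outside the output chain.
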AnimBlueprint /BRCosmetics/Characters/Player/Female/Medium/Bodies/F_MED_Parcel_Petal/Meshes/Parts/F_MED_Parcel_Petal_Skeleton_AnimBlueprint
target_skeleton=&/F_MED_Parcel_Petal_Skeleton
graphs [
 AnimGraph {nodes [
   n1 = Get IsPlayingIdle() -> (IsPlayingIdle:bool)
   n2 = "Transform (Modify) Bone - Bone: dyn_hair_back_l_base"!(ComponentPose, bAlphaBoolEnabled:bool=@n1, Translation="0,2,0", Rotation="0,0,0", Scale="1,1,1") -> (Pose) {
    alphaInputType=Bool
    rotationMode=BMM_Additive
    rotationSpace=BCS_BoneSpace
    translationMode=BMM_Additive
    translationSpace=BCS_BoneSpace
    alphaBoolBlend {blendOption=Linear}
    boneToModify {boneName=dyn_hair_back_l_base}
    translation {y=2}
   }
   n3 = "Transform (Modify) Bone - Bone: dyn_hair_back_m_base"!(ComponentPose=@n2, bAlphaBoolEnabled:bool=@n1, Translation="0,2,1", Rotation="-3.725,0.000,0.000", Scale="1,1,1") -> (Pose) {
    alphaInputType=Bool
    rotationMode=BMM_Additive
    rotationSpace=BCS_BoneSpace
    translationMode=BMM_Additive
    translationSpace=BCS_BoneSpace
    alphaBoolBlend {blendOption=Linear}
    boneToModify {boneName=dyn_hair_back_m_base}
    rotation {pitch=-3.725}
    translation {y=2, z=1}
   }
   "Transform (Modify) Bone - Bone: dyn_hair_front_r_base"!(ComponentPose=@n3, bAlphaBoolEnabled:bool=@n1, Translation="0,2,-2", Rotation="0,0,0", Scale="1,1,1") -> (Pose) {
    alphaInputType=Bool
    rotationMode=BMM_Additive
    rotationSpace=BCS_BoneSpace
    translationMode=BMM_Additive
    translationSpace=BCS_BoneSpace
    alphaBoolBlend {blendOption=Linear}
    boneToModify {boneName=dyn_hair_front_r_base}
    translation {y=2, z=-2}
   }
   n4 = Get ADS_Rotate() -> (ADS_Rotate:double)
   "Transform (Modify) Bone - Bone: dyn_hair_back_r_base"(Alpha=@n4, Translation="0,0,0", Rotation="0,0,-10", Scale="1,1,1") {
    rotationMode=BMM_Additive
    rotationSpace=BCS_BoneSpace
    scaleSpace=BCS_BoneSpace
    translationMode=BMM_Additive
    translationSpace=BCS_BoneSpace
    alphaBoolBlend {blendOption=Linear}
    alphaScaleBiasClamp {bInterpResult=true, interpSpeedDecreasing=4, interpSpeedIncreasing=6}
    boneToModify {boneName=dyn_hair_back_r_base}
    rotation {roll=-10}
   }
   n5 = "Transform (Modify) Bone - Bone: dyn_hair_side_r_base"(ComponentPose=Local To Component(), Alpha=@n4, Translation="0,0,0", Rotation="30,0,0.000", Scale="1,1,1") {
    rotationMode=BMM_Additive
    rotationSpace=BCS_BoneSpace
    scaleSpace=BCS_BoneSpace
    translationMode=BMM_Additive
    translationSpace=BCS_BoneSpace
    alphaBoolBlend {blendOption=Linear}
    alphaScaleBiasClamp {bInterpResult=true, interpSpeedDecreasing=4, interpSpeedIncreasing=6}
    boneToModify {boneName=dyn_hair_side_r_base}
    rotation {pitch=30}
   }
   "Transform (Modify) Bone - Bone: dyn_hair_front_r_base"!(ComponentPose=@n5, Alpha:float=@n4, Translation="0,0,0", Rotation="30,0,0.000", Scale="1,1,1") -> (Pose) {
    rotationMode=BMM_Additive
    rotationSpace=BCS_BoneSpace
    scaleMode=BMM_Replace
    scaleSpace=BCS_BoneSpace
    translationMode=BMM_Additive
    translationSpace=BCS_BoneSpace
    alphaBoolBlend {blendOption=Linear}
    alphaScaleBiasClamp {bInterpResult=true, interpSpeedDecreasing=4, interpSpeedIncreasing=6}
    boneToModify {boneName=dyn_hair_front_r_base}
    rotation {pitch=30}
   }
   RigidBody() {
    bClampLinearTranslationLimitToRefPose=true
    bOverrideWorldGravity=true
    overridePhysicsAsset=&/F_MED_Parcel_Petal_Physics
    simulationSpace=BaseBoneSpace
    alphaBoolBlend {blendOption=Linear}
    baseBoneRef {boneName=spine_05}
    componentAppliedLinearAccClamp=<6000 6000 6000>
    componentLinearAccScale=<0.15 0.15 0.45>
    componentLinearVelScale=<0.15 0.15 0.4>
    overrideWorldGravity {z=-980}
   }
   n6 = Anim Dynamics - Bone: dyn_hair_leaf_01(ComponentPose=Local To Component(LocalPose=Use cached pose 'CRCache_HairBackDyn'()), Alpha=0.75, ExternalForce="0,0,0") {
    angularDampingOverride=0.999
    bGravityOverrideInSimSpace=true
    bOverrideAngularDamping=true
    bOverrideLinearDamping=true
    bUseGravityOverride=true
    linearDampingOverride=0.999
    simulationSpace=RootRelative
    alphaBoolBlend {blendOption=Linear}
    boundBone {boneName=dyn_hair_leaf_01}
    componentAppliedLinearAccClamp=<3000 3000 3000>
    componentLinearAccScale=<0.25 0.25 0.5>
    componentLinearVelScale=<0.25 0.25 0.5>
    gravityOverride=<0.85 0.85 0.85>
    physicsBodyDefinitions [{
      sphereCollisionRadius=0
      boundBone {boneName=dyn_hair_leaf_01}
      boxExtents=<6 4 2>
      constraintSetup {angularLimitsMax=<0 20 10>, angularLimitsMin=<0 -10 -10>, angularTarget=<0 0 0>}
      localJointOffset=<2 0 0>
     }]
    relativeSpaceBone {boneName=spine_05}
   }
   n7 = Anim Dynamics - Bone: dyn_hair_leaf_02(ComponentPose=@n6, Alpha=0.75, ExternalForce="0,0,0") {
    angularDampingOverride=0.999
    bGravityOverrideInSimSpace=true
    bOverrideAngularDamping=true
    bOverrideLinearDamping=true
    bUseGravityOverride=true
    linearDampingOverride=0.999
    simulationSpace=RootRelative
    alphaBoolBlend {blendOption=Linear}
    boundBone {boneName=dyn_hair_leaf_02}
    componentAppliedLinearAccClamp=<3000 3000 3000>
    componentLinearAccScale=<0.25 0.25 0.5>
    componentLinearVelScale=<0.25 0.25 0.5>
    gravityOverride=<0.85 0.85 0.85>
    physicsBodyDefinitions [{
      sphereCollisionRadius=0
      boundBone {boneName=dyn_hair_leaf_02}
      boxExtents=<6 4 2>
      constraintSetup {angularLimitsMax=<0 20 10>, angularLimitsMin=<0 -1 -10>, angularTarget=<0 0 0>}
      localJointOffset=<2 0 0>
     }]
    relativeSpaceBone {boneName=spine_05}
   }
   n8 = Anim Dynamics - Bone: dyn_hair_leaf_03(ComponentPose=@n7, Alpha=0.75, ExternalForce="0,0,0") {
    angularDampingOverride=0.999
    bGravityOverrideInSimSpace=true
    bOverrideAngularDamping=true
    bOverrideLinearDamping=true
    bUseGravityOverride=true
    linearDampingOverride=0.999
    simulationSpace=RootRelative
    alphaBoolBlend {blendOption=Linear}
    boundBone {boneName=dyn_hair_leaf_03}
    componentAppliedLinearAccClamp=<3000 3000 3000>
    componentLinearAccScale=<0.25 0.25 0.5>
    componentLinearVelScale=<0.25 0.25 0.5>
    gravityOverride=<0.85 0.85 0.85>
    physicsBodyDefinitions [{
      sphereCollisionRadius=0
      boundBone {boneName=dyn_hair_leaf_03}
      boxExtents=<6 4 2>
      constraintSetup {angularLimitsMax=<0 20 10>, angularLimitsMin=<0 -10 -10>, angularTarget=<0 0 0>}
      localJointOffset=<2 0 0>
     }]
    relativeSpaceBone {boneName=spine_05}
   }
   n9 = Anim Dynamics - Bone: dyn_hair_leaf_04(ComponentPose=@n8, Alpha=0.75, ExternalForce="0,0,0") {
    angularDampingOverride=0.999
    bGravityOverrideInSimSpace=true
    bOverrideAngularDamping=true
    bOverrideLinearDamping=true
    bUseGravityOverride=true
    linearDampingOverride=0.999
    simulationSpace=RootRelative
    alphaBoolBlend {blendOption=Linear}
    boundBone {boneName=dyn_hair_leaf_04}
    componentAppliedLinearAccClamp=<3000 3000 3000>
    componentLinearAccScale=<0.25 0.25 0.5>
    componentLinearVelScale=<0.25 0.25 0.5>
    gravityOverride=<0.85 0.85 0.85>
    physicsBodyDefinitions [{
      sphereCollisionRadius=0
      boundBone {boneName=dyn_hair_leaf_04}
      boxExtents=<6 4 2>
      constraintSetup {angularLimitsMax=<0 20 10>, angularLimitsMin=<0 -10 -10>, angularTarget=<0 0 0>}
      localJointOffset=<2 0 0>
     }]
    relativeSpaceBone {boneName=spine_05}
   }
   Output Pose(Result=Component To Local(ComponentPose=@n9))
   n10 = Control Rig(Source=Copy Pose From Mesh() {bUseAttachedParent=true}) {
    alphaBoolBlend {blendOption=Linear}
    controlRigAssetReference {blueprintRigClass=&/F_MED_Parcel_Petal_CtrlRig.F_MED_Parcel_Petal_CtrlRig_C}
   }
   n11 = RigidBody(ComponentPose=Local To Component(LocalPose=@n10)) {
    bClampLinearTranslationLimitToRefPose=true
    bOverrideWorldGravity=true
    overridePhysicsAsset=&/F_MED_Parcel_Petal_Physics_Front
    simulationSpace=BaseBoneSpace
    alphaBoolBlend {blendOption=Linear}
    baseBoneRef {boneName=spine_05}
    componentAppliedLinearAccClamp=<6000 6000 6000>
    overrideWorldGravity {z=-980}
    simSpaceSettings {worldAlpha=0.3}
   }
   Save cached pose 'Control rig'(Pose=Component To Local(ComponentPose=@n11)) {cacheName=Control rig}
   Save cached pose 'CRCache_HairBackDyn'(Pose=Blend Poses by bool(
     BlendPose_0=Use cached pose 'CRCache_HairBack_AD'()
     BlendPose_1=Use cached pose 'CRCache_HairBack_RB'()
     bActiveValue=Get bIsSkydiveFloating()
    ) {blendType=Linear}) {cacheName=CRCache_HairBackDyn}
   n12 = RigidBody(ComponentPose=Local To Component(LocalPose=Use cached pose 'Control rig'())) {
    bClampLinearTranslationLimitToRefPose=true
    bOverrideWorldGravity=true
    overridePhysicsAsset=&/F_MED_Parcel_Petal_Physics_back
    simulationSpace=BaseBoneSpace
    alphaBoolBlend {blendOption=Linear}
    baseBoneRef {boneName=spine_05}
    componentAppliedLinearAccClamp=<6000 6000 6000>
    overrideWorldGravity {z=-980}
    simSpaceSettings {worldAlpha=0.3}
   }
   Save cached pose 'CRCache_HairBack_RB'(Pose=Component To Local(ComponentPose=@n12)) {cacheName=CRCache_HairBack_RB}
   n13 = "Constraint - dyn_hair_back_m_03 (PP)"!(ComponentPose=Local To Component(LocalPose=Use cached pose 'Control rig'()), Alpha=1, ConstraintWeights_0=0.6, ConstraintWeights_1:float=0.4) -> (Pose) {
    alphaBoolBlend {blendOption=Linear}
    boneToModify {boneName=dyn_hair_back_m_03}
    constraintSetup [{targetBone {boneName=head}}, {perAxis {bY=false, bZ=false}, targetBone {boneName=spine_05}}]
    constraintWeights [1,1]
   }
   n14 = Anim Dynamics - Bone: dyn_hair_back_l_01!(SimulationSpace:AnimPhysSimSpaceType Enum=World, ComponentPose=@n13, Alpha=0.3, GravityOverride:Vector="0,0,-980") -> (Pose) {
    angularDampingOverride=1
    angularSpringConstant=10000
    bAngularSpring=true
    bOverrideAngularDamping=true
    bUseGravityOverride=true
    gravityScale=0.35
    simulationSpace=World
    alphaBoolBlend {blendOption=Linear}
    boundBone {boneName=dyn_hair_back_l_01}
    componentAppliedLinearAccClamp=<2000 2000 2000>
    componentLinearAccScale=<0.5 0.5 0.5>
    componentLinearVelScale=<0.5 0.5 0.5>
    gravityOverride {z=-980}
    physicsBodyDefinitions [{
      boundBone {boneName=dyn_hair_back_l_01}
      boxExtents=<25 25 5>
      constraintSetup {angularLimitsMax=<0 2 0>, angularLimitsMin=<0 -20 0>, angularTarget=<1 0 0>}
      localJointOffset=<2 0 0>
     }]
    relativeSpaceBone {boneName=spine_02}
   }
   n15 = Anim Dynamics - Bone: dyn_hair_back_l_02!(SimulationSpace:AnimPhysSimSpaceType Enum, ComponentPose=@n14, Alpha=1, GravityOverride:Vector="0,0,-980") -> (Pose) {
    angularDampingOverride=1
    angularSpringConstant=20000
    bAngularSpring=true
    bGravityOverrideInSimSpace=true
    bOverrideAngularDamping=true
    simulationSpace=RootRelative
    alphaBoolBlend {blendOption=Linear}
    boundBone {boneName=dyn_hair_back_l_02}
    componentAppliedLinearAccClamp=<2000 2000 2000>
    componentLinearAccScale=<0.5 0.5 0.5>
    componentLinearVelScale=<0.5 0.5 0.5>
    gravityOverride {z=-980}
    physicsBodyDefinitions [{
      boundBone {boneName=dyn_hair_back_l_02}
      boxExtents=<25 25 5>
      constraintSetup {angularLimitsMax=<0 2 0>, angularLimitsMin=<0 -20 0>, angularTarget=<1 0 0>}
      localJointOffset=<2 0 0>
     }]
    relativeSpaceBone {boneName=spine_03}
   }
   n16 = Anim Dynamics - Bone: dyn_hair_back_l_03!(SimulationSpace:AnimPhysSimSpaceType Enum=World, ComponentPose=@n15, Alpha=1, GravityOverride:Vector="0,0,-980") -> (Pose) {
    angularDampingOverride=1
    angularSpringConstant=8000
    bAngularSpring=true
    bGravityOverrideInSimSpace=true
    bOverrideAngularDamping=true
    simulationSpace=World
    alphaBoolBlend {blendOption=Linear}
    alphaScaleBiasClamp {bInterpResult=true}
    boundBone {boneName=dyn_hair_back_l_03}
    componentAppliedLinearAccClamp=<2000 2000 2000>
    componentLinearAccScale=<0.5 0.5 0.5>
    componentLinearVelScale=<0.5 0.5 0.5>
    gravityOverride {z=-980}
    physicsBodyDefinitions [{
      boundBone {boneName=dyn_hair_back_l_03}
      boxExtents=<25 23 23>
      constraintSetup {angularLimitsMax=<0 0 20>, angularLimitsMin=<0 -70 -20>, angularTarget=<1 0 0>}
      localJointOffset=<2 0 0>
     }]
    relativeSpaceBone {boneName=root}
   }
   n17 = Anim Dynamics - Bone: dyn_hair_back_l_04!(SimulationSpace:AnimPhysSimSpaceType Enum=World, ComponentPose=@n16, Alpha=1, GravityOverride:Vector="0,0,-980") -> (Pose) {
    angularDampingOverride=1
    angularSpringConstant=5000
    bAngularSpring=true
    bGravityOverrideInSimSpace=true
    bOverrideAngularDamping=true
    simulationSpace=World
    alphaBoolBlend {blendOption=Linear}
    alphaScaleBiasClamp {bInterpResult=true}
    boundBone {boneName=dyn_hair_back_l_04}
    componentAppliedLinearAccClamp=<2000 2000 2000>
    componentLinearAccScale=<0.5 0.5 0.5>
    componentLinearVelScale=<0.5 0.5 0.5>
    gravityOverride {z=-980}
    physicsBodyDefinitions [{
      boundBone {boneName=dyn_hair_back_l_04}
      boxExtents=<25 23 23>
      constraintSetup {angularLimitsMax=<0 0 30>, angularLimitsMin=<0 -70 -30>, angularTarget=<1 0 0>}
      localJointOffset=<2 0 0>
     }]
    relativeSpaceBone {boneName=root}
   }
   n18 = Anim Dynamics - Bone: dyn_hair_back_m_01!(SimulationSpace:AnimPhysSimSpaceType Enum=World, ComponentPose=@n17, Alpha=0.3, GravityOverride:Vector="0,0,-980") -> (Pose) {
    angularDampingOverride=1
    angularSpringConstant=10000
    bAngularSpring=true
    bOverrideAngularDamping=true
    bUseGravityOverride=true
    gravityScale=0.35
    simulationSpace=World
    alphaBoolBlend {blendOption=Linear}
    boundBone {boneName=dyn_hair_back_m_01}
    componentAppliedLinearAccClamp=<2000 2000 2000>
    componentLinearAccScale=<0.5 0.5 0.5>
    componentLinearVelScale=<0.5 0.5 0.5>
    gravityOverride {z=-980}
    physicsBodyDefinitions [{
      boundBone {boneName=dyn_hair_back_m_01}
      boxExtents=<25 25 5>
      constraintSetup {angularLimitsMax=<0 2 0>, angularLimitsMin=<0 -20 0>, angularTarget=<1 0 0>}
      localJointOffset=<2 0 0>
     }]
    relativeSpaceBone {boneName=spine_02}
   }
   n19 = Anim Dynamics - Bone: dyn_hair_back_m_02!(SimulationSpace:AnimPhysSimSpaceType Enum, ComponentPose=@n18, Alpha=1, GravityOverride:Vector="0,0,-980") -> (Pose) {
    angularDampingOverride=1
    angularSpringConstant=20000
    bAngularSpring=true
    bGravityOverrideInSimSpace=true
    bOverrideAngularDamping=true
    simulationSpace=RootRelative
    alphaBoolBlend {blendOption=Linear}
    boundBone {boneName=dyn_hair_back_m_02}
    componentAppliedLinearAccClamp=<2000 2000 2000>
    componentLinearAccScale=<0.5 0.5 0.5>
    componentLinearVelScale=<0.5 0.5 0.5>
    gravityOverride {z=-980}
    physicsBodyDefinitions [{
      boundBone {boneName=dyn_hair_back_m_02}
      boxExtents=<25 25 5>
      constraintSetup {angularLimitsMax=<0 2 0>, angularLimitsMin=<0 -20 0>, angularTarget=<1 0 0>}
      localJointOffset=<2 0 0>
     }]
    relativeSpaceBone {boneName=spine_03}
   }
   n20 = Anim Dynamics - Bone: dyn_hair_back_m_03!(SimulationSpace:AnimPhysSimSpaceType Enum=World, ComponentPose=@n19, Alpha=1, GravityOverride:Vector="0,0,-980") -> (Pose) {
    angularDampingOverride=1
    angularSpringConstant=8000
    bAngularSpring=true
    bGravityOverrideInSimSpace=true
    bOverrideAngularDamping=true
    simulationSpace=World
    alphaBoolBlend {blendOption=Linear}
    alphaScaleBiasClamp {bInterpResult=true}
    boundBone {boneName=dyn_hair_back_m_03}
    componentAppliedLinearAccClamp=<2000 2000 2000>
    componentLinearAccScale=<0.5 0.5 0.5>
    componentLinearVelScale=<0.5 0.5 0.5>
    gravityOverride {z=-980}
    physicsBodyDefinitions [{
      boundBone {boneName=dyn_hair_back_m_03}
      boxExtents=<25 23 23>
      constraintSetup {angularLimitsMax=<0 0 20>, angularLimitsMin=<0 -70 -20>, angularTarget=<1 0 0>}
      localJointOffset=<2 0 0>
     }]
    relativeSpaceBone {boneName=root}
   }
   n21 = Anim Dynamics - Bone: dyn_hair_back_m_04!(SimulationSpace:AnimPhysSimSpaceType Enum=World, ComponentPose=@n20, Alpha=1, GravityOverride:Vector="0,0,-980") -> (Pose) {
    angularDampingOverride=1
    angularSpringConstant=5000
    bAngularSpring=true
    bGravityOverrideInSimSpace=true
    bOverrideAngularDamping=true
    simulationSpace=World
    alphaBoolBlend {blendOption=Linear}
    alphaScaleBiasClamp {bInterpResult=true}
    boundBone {boneName=dyn_hair_back_m_04}
    componentAppliedLinearAccClamp=<2000 2000 2000>
    componentLinearAccScale=<0.5 0.5 0.5>
    componentLinearVelScale=<0.5 0.5 0.5>
    gravityOverride {z=-980}
    physicsBodyDefinitions [{
      boundBone {boneName=dyn_hair_back_m_04}
      boxExtents=<25 23 23>
      constraintSetup {angularLimitsMax=<0 0 30>, angularLimitsMin=<0 -70 -30>, angularTarget=<1 0 0>}
      localJointOffset=<2 0 0>
     }]
    relativeSpaceBone {boneName=root}
   }
   n22 = Anim Dynamics - Bone: dyn_hair_back_r_01!(SimulationSpace:AnimPhysSimSpaceType Enum=World, ComponentPose=@n21, Alpha=0.3, GravityOverride:Vector="0,0,-980") -> (Pose) {
    angularDampingOverride=1
    angularSpringConstant=10000
    bAngularSpring=true
    bOverrideAngularDamping=true
    bUseGravityOverride=true
    gravityScale=0.35
    simulationSpace=World
    alphaBoolBlend {blendOption=Linear}
    boundBone {boneName=dyn_hair_back_r_01}
    componentAppliedLinearAccClamp=<2000 2000 2000>
    componentLinearAccScale=<0.5 0.5 0.5>
    componentLinearVelScale=<0.5 0.5 0.5>
    gravityOverride {z=-980}
    physicsBodyDefinitions [{
      boundBone {boneName=dyn_hair_back_r_01}
      boxExtents=<25 25 5>
      constraintSetup {angularLimitsMax=<0 2 0>, angularLimitsMin=<0 -20 0>, angularTarget=<1 0 0>}
      localJointOffset=<2 0 0>
     }]
    relativeSpaceBone {boneName=spine_02}
   }
   n23 = Anim Dynamics - Bone: dyn_hair_back_r_02!(SimulationSpace:AnimPhysSimSpaceType Enum, ComponentPose=@n22, Alpha=1, GravityOverride:Vector="0,0,-980") -> (Pose) {
    angularDampingOverride=1
    angularSpringConstant=20000
    bAngularSpring=true
    bGravityOverrideInSimSpace=true
    bOverrideAngularDamping=true
    simulationSpace=RootRelative
    alphaBoolBlend {blendOption=Linear}
    boundBone {boneName=dyn_hair_back_r_02}
    componentAppliedLinearAccClamp=<2000 2000 2000>
    componentLinearAccScale=<0.5 0.5 0.5>
    componentLinearVelScale=<0.5 0.5 0.5>
    gravityOverride {z=-980}
    physicsBodyDefinitions [{
      boundBone {boneName=dyn_hair_back_r_02}
      boxExtents=<25 25 5>
      constraintSetup {angularLimitsMax=<0 2 0>, angularLimitsMin=<0 -20 0>, angularTarget=<1 0 0>}
      localJointOffset=<2 0 0>
     }]
    relativeSpaceBone {boneName=spine_03}
   }
   n24 = Anim Dynamics - Bone: dyn_hair_back_r_03!(SimulationSpace:AnimPhysSimSpaceType Enum=World, ComponentPose=@n23, Alpha=1, GravityOverride:Vector="0,0,-980") -> (Pose) {
    angularDampingOverride=1
    angularSpringConstant=8000
    bAngularSpring=true
    bGravityOverrideInSimSpace=true
    bOverrideAngularDamping=true
    simulationSpace=World
    alphaBoolBlend {blendOption=Linear}
    alphaScaleBiasClamp {bInterpResult=true}
    boundBone {boneName=dyn_hair_back_r_03}
    componentAppliedLinearAccClamp=<2000 2000 2000>
    componentLinearAccScale=<0.5 0.5 0.5>
    componentLinearVelScale=<0.5 0.5 0.5>
    gravityOverride {z=-980}
    physicsBodyDefinitions [{
      boundBone {boneName=dyn_hair_back_r_03}
      boxExtents=<25 23 23>
      constraintSetup {angularLimitsMax=<0 0 20>, angularLimitsMin=<0 -70 -20>, angularTarget=<1 0 0>}
      localJointOffset=<2 0 0>
     }]
    relativeSpaceBone {boneName=root}
   }
   n25 = Anim Dynamics - Bone: dyn_hair_back_r_04!(SimulationSpace:AnimPhysSimSpaceType Enum=World, ComponentPose=@n24, Alpha=1, GravityOverride:Vector="0,0,-980") -> (Pose) {
    angularDampingOverride=1
    angularSpringConstant=5000
    bAngularSpring=true
    bGravityOverrideInSimSpace=true
    bOverrideAngularDamping=true
    simulationSpace=World
    alphaBoolBlend {blendOption=Linear}
    alphaScaleBiasClamp {bInterpResult=true}
    boundBone {boneName=dyn_hair_back_r_04}
    componentAppliedLinearAccClamp=<2000 2000 2000>
    componentLinearAccScale=<0.5 0.5 0.5>
    componentLinearVelScale=<0.5 0.5 0.5>
    gravityOverride {z=-980}
    physicsBodyDefinitions [{
      boundBone {boneName=dyn_hair_back_r_04}
      boxExtents=<25 23 23>
      constraintSetup {angularLimitsMax=<0 0 30>, angularLimitsMin=<0 -70 -30>, angularTarget=<1 0 0>}
      localJointOffset=<2 0 0>
     }]
    relativeSpaceBone {boneName=root}
   }
   Save cached pose 'CRCache_HairBack_AD'(Pose=Component To Local(ComponentPose=@n25)) {cacheName=CRCache_HairBack_AD}
  ]}
 EventGraph {nodes [
   Event OnCharacterCustomizationLoaded()
   > Cast To FortPlayerPawn(Object=TryGetFortPawn())
   n1 = CheckForBackbling()
   > Set is_WearingCape(is_WearingCape:bool=@n1.bIsBackblingLong) -> (Output_Get:bool)
   > Set isWearingPackBool(isWearingPackBool:bool=OR Boolean!(A=@n1, B=@n1.bIsBackblingACape, C:bool=@n1.bIsBackblingLong) -> (ReturnValue)) -> (Output_Get:bool)
   n2 = Event BlueprintProcessAnimInputEvent()
   > n3 = Branch(Condition=IsNamed(self=@n2.AnimInputEvent, InEventName=aie.Player.Zipline.Entered))
   > Set OnZipline(OnZipline:bool=true) -> (Output_Get:bool)
   Branch(execute=@n3.else, Condition=IsNamed(self=@n2.AnimInputEvent, InEventName=aie.Player.Zipline.Exited))
   > Set OnZipline(OnZipline=false)
   Event BlueprintUpdateAnimation()
   > Get!(execute:Exec) -> (then:Exec, else:Exec, Skel_Mesh:Skeletal Mesh Component Object Reference)
   > Get!(execute) -> (then, else, Anim_Inst:Anim Instance Object Reference)
   > n4 = Sequence!(execute) -> (then_0, then_1, then_2:Exec, then_3:Exec, then_4:Exec, then_5:Exec)
   n5 = BreakTransform(InTransform=GetDeltaTransformFromRefPose(self=GetOwnerSkeletalMeshComponent(), BoneName=head, BaseName=spine_05))
   > Set Head_Rot_Back(Head_Rot_Back:double=MapRangeClamped(Value=BreakRotator(InRot=@n5.Rotation).Yaw, InRangeA=0, InRangeB=-48)) -> (Output_Get:double)
   n6 = AND Boolean(A=OR Boolean(A=Get bIsSkydiveDiveMode(), B=Get bIsSkyDiving(), C=Get bIsSkydiveFloating()), B=NOT Boolean(A=Get bIsParachuteOpened()))
   n7 = Get TacticalJump() -> (TacticalJump:bool)
   > Set HairSkydivePose(HairSkydivePose:double="To Float (Boolean)"(InBool=OR Boolean!(A=@n7, B=@n6) -> (ReturnValue))) -> (Output_Get:double)
   > n8 = Branch(Condition=OR Boolean!(A=Get bIsSkydiveDiveMode(), B=Get bIsSkyDiving(), C:bool=Get bIsSkydiveFloating(), D:bool=Get bIsParachuteOpened()) -> (ReturnValue))
   > Set BackHairDynAlpha(BackHairDynAlpha:double=0) -> (Output_Get:double)
   n9 = Branch(execute=@n8.else, Condition=OR Boolean!(A=Get bIsTargeting(), B=Get bIsCrouching()) -> (ReturnValue))
   > Set BackHairDynAlpha(BackHairDynAlpha=0.3)
   > Set UpperHairDyn(UpperHairDyn:double=0.1) -> (Output_Get:double)
   Set BackHairDynAlpha(execute=@n9.else, BackHairDynAlpha=1)
   > Set UpperHairDyn(UpperHairDyn=0.15)
   > n10 = Branch(Condition=Get bIsInFrontEnd())
   > Set BackHairDynAlpha(BackHairDynAlpha=0.45)
   Set BackHairDynAlpha(execute=@n10.else, BackHairDynAlpha=1)
   n11 = Branch(
    execute=@n4.then_3
    Condition=OR Boolean!(A=Get bIsJumping(), B=Get bIsFalling(), C:bool=Get bIsInFrontEnd(), D:bool=Get bIsDBNO()) -> (ReturnValue)
   )
   > Set JumpFix(JumpFix:double=0) -> (Output_Get:double)
   Set JumpFix(execute=@n11.else, JumpFix=1)
   n12 = Get bIsJumping()
   n13 = Get bIsFalling()
   Set TacticalJump(execute=@n4.then_4, TacticalJump:bool=AND Boolean(A=Get bIsTacticalSprinting(), B=OR Boolean!(A=@n12, B=@n13) -> (ReturnValue))) -> (Output_Get:bool)
   n14 = OR Boolean!(A=Get bIsSkydiveDiveMode(), B=Get bIsSkyDiving(), C:bool=Get bIsSkydiveFloating(), D:bool=Get bIsParachuteOpened()) -> (ReturnValue)
   > Set isInAir(isInAir:double="To Float (Boolean)"(InBool=@n14)) -> (Output_Get:double)
   n15 = NOT Boolean(A=@n14)
   n16 = NOT Boolean(A=Get TacticalJump())
   > Set AirFloatingAlpha(AirFloatingAlpha:double="To Float (Boolean)"(InBool=OR Boolean!(A=@n15, B=@n16) -> (ReturnValue))) -> (Output_Get:double)
   > n17 = Branch(Condition=OR Boolean(A=Get isWearingPackBool() -> (isWearingPackBool:bool), B=Get bIsDBNO()))
   > Set BackADS_Fix(BackADS_Fix:double=0) -> (Output_Get:double)
   Set BackADS_Fix(execute=@n17.else, BackADS_Fix=1)
   n18 = Gravity Override State Function DEPRECATED(
    execute=@n4.then_5
    CharacterPartAnimInstance=Self-Reference()
    GameTimeInSeconds=GetGameTimeInSeconds()
    GravityOverrideParams=Get Gravity Override Params()
   )
   > "Set Gravity Override "("Gravity Override ":Vector=@n18) -> (Output_Get:Vector)
   > n19 = Branch(Condition=Get OnZipline() -> (OnZipline:bool))
   > Set Simulation Space Dyncamis Hair(Simulation Space Dyncamis Hair:AnimPhysSimSpaceType Enum=Component) -> (Output_Get:AnimPhysSimSpaceType Enum)
   Set Simulation Space Dyncamis Hair(execute=@n19.else, Simulation Space Dyncamis Hair=World)
   n20 = "In Range (Float)"!(Value:double=Get Head_Rot_Back() -> (Head_Rot_Back:double), Min:double=0.2, Max:double=1, InclusiveMin:bool=true, InclusiveMax:bool=true) -> (ReturnValue:bool)
   n21 = Get bIsTargeting()
   Set ADS_Rotate(execute=@n4.then_1, ADS_Rotate:double=SelectFloat(A=1, B=0, bPickA=AND Boolean(A=@n20, B=@n21))) -> (Output_Get:double)
   n22 = Get Anim_Inst() -> (Anim_Inst:Anim Instance Object Reference)
   Is Valid(exec=@n4.then_2, InputObject=@n22)
   > Set IsPlayingIdle(IsPlayingIdle:bool="float > float"(A=GetCurveValue(self=@n22, CurveName=FrontEnd_Idle_Bypass), B=0) {is=title K2Node_CallFunction}) -> (Output_Get:bool)
   Event BlueprintInitializeAnimation()
   n23 = GetOwnerSkeletalMeshComponent()
   > Is Valid(InputObject=@n23)
   > Set Skel_Mesh(Skel_Mesh=@n23)
   > n24 = Cast To FortPlayerPawn(Object=GetOwningActor())
   > Is Valid(InputObject=@n24)
   > n25 = Set Pawn(Pawn:Fort Player Pawn Object Reference=@n24) -> (Output_Get:Fort Player Pawn Object Reference)
   n26 = GetAnimInstance(self=Get Mesh(self=@n25))
   > Is Valid(InputObject=@n26)
   > Set Anim_Inst(Anim_Inst:Anim Instance Object Reference=@n26) -> (Output_Get:Anim Instance Object Reference)
   Set MoveHair(MoveHair:bool=false) -> (Output_Get:bool)
   > Branch(Condition=Get MoveHair() -> (MoveHair:bool))
   > Set Alpha(Alpha:double=0) -> (Output_Get:double)
   TryGetPawnOwner()
   "float > float"(A=GetCurveValue(CurveName=Disable_Custom), B=0)
  ]}
]
variables [
 ADS_Rotate {type=number}
 AirFloatingAlpha {type=number}
 Alpha {type=number}
 Anim_Inst {type=AnimInstance, default=null}
 BackADS_Fix {type=number}
 BackHairDynAlpha {type=number}
 "Gravity Override " {type=Vector}
 Gravity Override Params {type=GravityOverrideParamsStruct}
 HairSkydivePose {type=number}
 Head_Rot_Back {type=number}
 is_WearingCape {type=boolean}
 isInAir {type=number}
 IsPlayingIdle {type=boolean}
 isWearingPackBool {type=boolean}
 JumpFix {type=number}
 MoveHair {type=boolean}
 OnZipline {type=boolean}
 Pawn {type=FortPlayerPawn}
 Simulation Space Dyncamis Hair {type=AnimPhysSimSpaceType}
 Skel_Mesh {type=SkeletalMeshComponent}
 TacticalJump {type=boolean}
 UpperHairDyn {type=number}
]
